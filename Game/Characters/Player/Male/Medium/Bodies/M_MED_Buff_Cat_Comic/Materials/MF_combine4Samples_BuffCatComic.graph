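# Material function in the Buff Cat Comic character's materials that turns a base value and four samples into an outline mask.
# It sums biased base-minus-sample differences, divides by positive and negative Divisor, then sharpens with Scale and clamps.
# UseCenteredLines weights the opposite-sign edge for lines on both sides of a discontinuity.
# Low and high filter outputs drop to zero when any input, described as roughness, crosses its threshold.
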MaterialFunction /Game/Characters/Player/Male/Medium/Bodies/M_MED_Buff_Cat_Comic/Materials/MF_combine4Samples_BuffCatComic
graph {nodes [
  Constant() {desc=1, r=2}
  n1 = FunctionInput() {bUsePreviewValueAsDefault=true, inputName=Base, inputType=FunctionInput_Scalar}
  n2 = Add(A=FunctionInput() {bUsePreviewValueAsDefault=true, inputName=BaseSampleBias, inputType=FunctionInput_Scalar}, B=@n1)
  n3 = FunctionInput() {bUsePreviewValueAsDefault=true, inputName=Sample1, inputType=FunctionInput_Scalar, sortPriority=1}
  n4 = FunctionInput() {bUsePreviewValueAsDefault=true, inputName=Sample2, inputType=FunctionInput_Scalar, sortPriority=2}
  n5 = FunctionInput() {bUsePreviewValueAsDefault=true, inputName=Sample3, inputType=FunctionInput_Scalar, sortPriority=3}
  n6 = FunctionInput() {bUsePreviewValueAsDefault=true, inputName=Sample4, inputType=FunctionInput_Scalar, sortPriority=4}
  n7 = Add(A=Add(A=Subtract(A=@n2, B=@n3), B=Subtract(A=@n2, B=@n4)), B=Add(A=Subtract(A=@n2, B=@n5), B=Subtract(A=@n2, B=@n6)))
  n8 = FunctionInput() {bUsePreviewValueAsDefault=true, inputName=Divisor, inputType=FunctionInput_Scalar, sortPriority=5, previewValue=<-64 0 0 1>}
  n9 = Divide(
   A=MaterialFunctionCall(X=@n7, Y=@n7) {materialFunction=&MakeFloat2}.Result
   B=MaterialFunctionCall(X=@n8, Y=Multiply(A=@n8) {constB=-1}) {materialFunction=&MakeFloat2}.Result
  )
  n10 = FunctionInput() {bUsePreviewValueAsDefault=true, inputName=Scale, inputType=FunctionInput_Scalar, previewValue=<2 0 0 1>}
  n11 = Multiply(A=OneMinus(None=Clamp(None=@n9) {desc=depth outline}), B=@n10)
  n12 = MaterialFunctionCall(Float2=OneMinus(None=Clamp(None=@n11) {desc=depth outline})) {materialFunction=&BreakOutFloat2Components}
  n13 = ScalarParameter() {desc=Used to draw outlines on both sides of a discontinuity, parameterName=UseCenteredLines}
  FunctionOutput(None=Max(A=@n12.R, B=Multiply(A=@n12.G, B=@n13)))
  n14 = Min(A=@n1, B=@n3)
  n15 = Min(A=@n4, B=@n5)
  n16 = Subtract(
   A=Min(A=Min(A=@n14, B=@n15), B=@n6)
   B=FunctionInput() {bUsePreviewValueAsDefault=true, inputName=Low Filter Threshold, inputType=FunctionInput_Scalar, sortPriority=1000}
  )
  FunctionOutput(None=Saturate(None=Sign(None=@n16))) {
   description=Ouputs outlne mask where the roughness value in one of the samples is lower than Input Low Filter Threshold
   outputName=Low Pass Filter
   sortPriority=20
  }
  n17 = Max(A=Max(A=Max(A=@n1, B=@n3), B=Max(A=@n4, B=@n5)), B=@n6)
  n18 = FunctionInput() {bUsePreviewValueAsDefault=true, inputName=High Filter Threshold, inputType=FunctionInput_Scalar, sortPriority=1050, previewValue=<1 0 0 1>}
  n19 = Subtract(A=OneMinus(None=Saturate(None=@n17)), B=OneMinus(None=@n18))
  FunctionOutput(None=Saturate(None=Sign(None=@n19))) {
   description=Ouputs outlne mask where the roughness value in one of the samples is higher than Input High Filter Threshold
   outputName=High Pass Filter
   sortPriority=30
  }
 ]}
