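# Translucent unlit outline material for the Buff Cat comic character, drawing screen-space inner lines that are dark or, optionally, partly white.
# Four-neighbor SceneDepth edge detection, optionally combined with roughness GBuffer edges, drives opacity, with distance-scaled width and GBuffer masking.
# An optional white-line mask uses edge filters and a pre-skinned biped mask.
# Shading-path switches give mobile and forward simpler masks, separate colors and depth lines only behind a static switch.
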Material /Game/Characters/Player/Male/Medium/Bodies/M_MED_Buff_Cat_Comic/Materials/M_BlackNWhiteOutline
graph {
 properties {
  bCastRayTracedShadows=false
  bUseMaterialAttributes=true
  bUsedWithMeshDeformer=true
  bUsedWithMorphTargets=true
  bUsedWithSkeletalMesh=true
  blendMode=BLEND_Translucent
  pixelDepthOffsetMode=PDOM_Legacy
  shadingModel=MSM_Unlit
  twoSided=true
  usedShadingModels=Unlit
 }
 nodes [
  n1 = VectorParameter() {parameterName=OutlineColor, defaultValue=<rgba 0 0 0 1>}
  n2 = Multiply(
   A=ComponentMask(None=SceneTexture() {sceneTextureId=PPI_DiffuseColor}.Color) {b=true, g=true, r=true}
   B=VectorParameter() {parameterName=OutlineDakenMultiplier, defaultValue=<rgba 0.1 0.1 0.1 0>}.RGB
  )
  n3 = VectorParameter() {parameterName=MobileOutlineColor}
  n4 = LinearInterpolate(
   A=ShadingPathSwitch(Default=StaticSwitchParameter(True=@n1.RGB, False=@n2) {parameterName=UseOutlineColor}, Forward=@n3.RGB, Mobile=@n3.RGB)
   B=VectorParameter() {parameterName=MobileInsideOutlineColor, defaultValue=<rgba 0.9 0.9 0.9 0>}.RGB
   Alpha=NamedRerouteUsage()
  ) {constB=0}
  n5 = SceneTexture() {sceneTextureId=PPI_SubsurfaceColor}
  n6 = Subtract(A=@n5.Color, B=SquareRoot(None=Constant() {r=0.001})) {constB=0.001}
  n7 = Multiply(A=Sign(None=@n5.Color), B=OneMinus(None=Sign(None=@n6)))
  n8 = ComponentMask(None=@n7) {g=true}
  n9 = Constant() {r=1}
  n10 = Floor(None=SceneTexture() {sceneTextureId=PPI_Opacity}.Color)
  n11 = PixelDepth()
  n12 = SceneDepth() {desc=SceneDepth works on mobile whereas SceneTexture doesn't}
  n13 = ScalarParameter() {defaultValue=3, group=Cartoon, parameterName=CRTNO_Outline-InnerMaskBias}
  n14 = Divide(A=ObjectRadius(), B=ScalarParameter() {defaultValue=97, parameterName=ObjectScale})
  n15 = Saturate(None=Add(A=Subtract(A=@n11.R, B=@n12.R), B=Multiply(A=@n13, B=@n14)))
  n16 = ShadingPathSwitch(Default=ComponentMask(None=QualitySwitch(Default=@n7, Low=@n10, High=@n10, Medium=@n10, Epic=@n7)) {r=true}, Forward=@n15, Mobile=@n15)
  n17 = PixelDepth()
  n18 = ScalarParameter() {defaultValue=50, group=Cartoon, parameterName=CRTNO_InnerLine_FadeStart}
  n19 = Clamp(None=Divide(A=Subtract(A=@n17.R, B=@n18), B=ScalarParameter() {defaultValue=500, group=Cartoon, parameterName=CRTNO_InnerLine_FadeDistance}))
  n20 = ShadingPathSwitch(
   Default=LinearInterpolate(
    A=ScalarParameter() {defaultValue=16, parameterName=SampleBias_Min}
    B=ScalarParameter() {defaultValue=16, parameterName=SampleBias}
    Alpha=@n19
   )
   Mobile=LinearInterpolate(
    A=ScalarParameter() {defaultValue=1, parameterName=MobileSampleBias_Min}
    B=ScalarParameter() {defaultValue=16, parameterName=Mobile_SampleBias}
    Alpha=@n19
   )
  )
  n21 = ComponentMask(None=ScreenPosition().ViewportUV) {g=true, r=true}
  n22 = Subtract(A=Divide(A=ViewSize(), B=Constant() {r=320}), B=Constant() {r=2}) {constB=0}
  n23 = Subtract(A=PixelDepth().R, B=ScalarParameter() {defaultValue=50, group=Cartoon, parameterName=CRTNO_Outline-InnerScaleStart})
  n24 = ScalarParameter() {defaultValue=500, group=Cartoon, parameterName=CRTNO_Outline-InnerScaleDistance}
  n25 = LinearInterpolate(
   A=ScalarParameter() {defaultValue=2, group=Cartoon, parameterName=CRTNO_Outline-InnerSize1}
   B=ScalarParameter() {defaultValue=0.75, group=Cartoon, parameterName=CRTNO_Outline-InnerSize2}
   Alpha=Clamp(None=Divide(A=@n23, B=@n24))
  )
  n26 = Multiply(A=@n25, B=ScalarParameter() {defaultValue=1.5, group=Cartoon, parameterName=CRTNO_Outline-InnerSizeMobileMultiplier})
  n27 = Multiply(
   A=Clamp(None=Ceil(None=@n22)) {maxDefault=2, minDefault=1}
   B=Multiply(A=SceneTexelSize(), B=ShadingPathSwitch(Default=@n25, Forward=@n26, Mobile=@n26))
  ) {constA=2}
  n28 = Constant()
  n29 = AppendVector(A=ComponentMask(None=@n27) {desc=x, r=true}, B=@n28)
  n30 = Add(A=@n21, B=@n29) {desc=x+}
  n31 = SceneDepth(UVs=@n30)
  n32 = AppendVector(A=@n28, B=ComponentMask(None=@n27) {desc=y, g=true}) {desc=y+}
  n33 = Add(A=@n21, B=@n32) {desc=y+}
  n34 = SceneDepth(UVs=@n33)
  n35 = Constant() {r=-1}
  n36 = Add(A=@n21, B=Multiply(A=@n29, B=@n35)) {desc=x-}
  n37 = SceneDepth(UVs=@n36)
  n38 = Add(A=@n21, B=Multiply(A=@n35, B=@n32)) {desc=y-}
  n39 = SceneDepth(UVs=@n38)
  n40 = LinearInterpolate(
   A=ScalarParameter() {defaultValue=-1, parameterName=Divisor_Min}
   B=ScalarParameter() {defaultValue=-16, parameterName=Divisor_max}
   Alpha=@n19
  )
  n41 = LinearInterpolate(
   A=ScalarParameter() {defaultValue=-1, parameterName=MobileDivisor_Min}
   B=ScalarParameter() {defaultValue=-16, parameterName=MobileDivisor_max}
   Alpha=@n19
  )
  n42 = ComponentMask(None=MaterialFunctionCall!(
    Base=@n12.R
    BaseSampleBias=@n20
    Scale=ScalarParameter() {defaultValue=4, parameterName=Scale}
    Sample1=@n31.R
    Sample2=@n34.R
    Sample3=@n37.R
    Sample4=@n39.R
    Divisor=ShadingPathSwitch(Default=@n40, Mobile=@n41)
    UseCenteredLines
    UseRGB
   ) {materialFunction=/Game/Packages/jl_tests/combine4Samples}.Result) {r=true}
  n43 = MaterialFunctionCall!(
   Base=SceneTexture() {sceneTextureId=PPI_Roughness}.Color
   BaseSampleBias=Constant()
   Scale=Constant() {r=4}
   Sample1=SceneTexture(UVs=@n30) {sceneTextureId=PPI_Roughness}.Color
   Sample2=SceneTexture(UVs=@n33) {sceneTextureId=PPI_Roughness}.Color
   Sample3=SceneTexture(UVs=@n36) {sceneTextureId=PPI_Roughness}.Color
   Sample4=SceneTexture(UVs=@n38) {sceneTextureId=PPI_Roughness}.Color
   Divisor=ScalarParameter() {defaultValue=-0.04, group=RoughnessGBufferLines, parameterName=InnerLinesDivisor}
   Low Filter Threshold
   High Filter Threshold=ScalarParameter() {defaultValue=0.6, parameterName=Outline High Filter}
  ) {materialFunction=&/MF_combine4Samples_BuffCatComic}
  n44 = ComponentMask(None=@n43.Result) {r=true}
  n45 = StaticSwitchParameter(True=ShadingPathSwitch(Default=Max(A=@n42, B=@n44), Forward=@n42, Mobile=@n42), False=@n42) {group=RoughnessGBufferLines, parameterName=UseRoughnessGBufferOutlines}
  n46 = Multiply(A=Saturate(None=Multiply(A=@n16, B=@n45)), B=ScalarParameter() {defaultValue=1, group=Cartoon, parameterName=CRTNO_Outline-InnerAmount})
  n47 = Constant()
  n48 = ShadingPathSwitch(Default=@n46, Forward=@n47, Mobile=@n47)
  n49 = Multiply(
   A=ShadingPathSwitch(Default=OneMinus(None=@n8), Forward=@n9, Mobile=@n9)
   B=Saturate(None=StaticSwitchParameter(True=@n46, False=@n48) {group=Cartoon, parameterName=UseDepthBufferOutlinesOnMobile})
  )
  n50 = Multiply(A=VertexNormalWS(), B=ScalarParameter() {parameterName=ScaleMesh})
  MaterialFunctionCall(In=SetMaterialAttributes!(
    MaterialAttributes
    Emissive Color=NamedRerouteDeclaration(None=@n4) {name=emissive, nodeColor=<rgba 0.5938 0 1 1>}
    Opacity=NamedRerouteDeclaration(None=@n49) {name=opacity, nodeColor=<rgba 1 0 0.9141 1>}
    World Position Offset=NamedRerouteDeclaration(None=@n50) {name=wpo, nodeColor=<rgba 0.7578 1 0 1>}
   ) {attributeSetTypes [#1, #2, #3]}) {materialFunction=&MF_TechArtOverrides}
  n51 = Constant() {r=1}
  n52 = ShadingPathSwitch(Default=ComponentMask(None=@n43.High Pass Filter) {r=true}, Forward=@n51, Mobile=@n51)
  n53 = ScalarParameter() {defaultValue=1}
  n54 = ScalarParameter() {defaultValue=-1, parameterName=Param_1}
  n55 = OneMinus(None=Saturate(None=MaterialFunctionCall(Base=@n12.R, BaseSampleBias=@n20, Scale=@n53, Sample1=@n31.R, Sample2=@n34.R, Sample3=@n37.R, Sample4=@n39.R, Divisor=@n54) {materialFunction=&combine4Samples}.Result))
  n56 = ScalarParameter() {defaultValue=80, parameterName=Force Black Height}
  n57 = ScalarParameter() {defaultValue=5, parameterName=Force Black Falloff}
  n58 = VertexInterpolator(VS=PreSkinnedPosition())
  n59 = SmoothStep(Min=Subtract(A=@n56, B=@n57), Max=Add(A=@n56, B=@n57), Value=ComponentMask(None=@n58.PS) {b=true})
  n60 = ComponentMask(None=@n58.PS) {b=true, r=true}
  n61 = ScalarParameter() {defaultValue=70, parameterName=Hand Mask X}
  n62 = ScalarParameter() {defaultValue=70, parameterName=Hand Mask Z}
  n63 = OneMinus(None=SphereMask(
    A=Abs(None=@n60)
    B=AppendVector(A=@n61, B=@n62)
    Radius=ScalarParameter() {defaultValue=40, parameterName=Param_2}
    Hardness=Constant() {r=1}
   ))
  n64 = SphereMask(A=@n58.PS, B=Constant3Vector() {constant=<rgba 0 -3 67 0>}, Radius=Constant() {r=28}, Hardness=Constant() {r=0.8})
  n65 = StaticSwitchParameter(True=Min(A=Multiply(A=@n59, B=@n63), B=OneMinus(None=@n64)), False=Constant() {r=1}) {parameterName=Use Biped Mask}
  n66 = LinearInterpolate(B=Multiply(A=Saturate(None=@n52), B=Multiply(A=@n55, B=@n65)) {constA=1}, Alpha=ScalarParameter() {parameterName=EnableWhiteOutline})
  NamedRerouteDeclaration(None=@n66) {name=Outline Color Mask, nodeColor=<rgba 0 0.4375 1 1>}
  Reroute(None=@n66)
 ]
}
