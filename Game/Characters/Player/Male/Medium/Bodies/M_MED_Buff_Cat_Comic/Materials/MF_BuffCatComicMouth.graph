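# Material function that composites a comic mouth from a faces flipbook texture onto the Buff Cat Comic player character's material.
# UVs are shifted by offset parameters and scaled about their center, then sampled through the weapon-skin CustomNumbers_SubUV function at a hardcoded constant frame, so it does not animate here.
# Texture alpha times the scaled UVs' 0-1 mask blends the mouth into both BaseColor and EmissiveColor of the incoming material attributes.
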MaterialFunction /Game/Characters/Player/Male/Medium/Bodies/M_MED_Buff_Cat_Comic/Materials/MF_BuffCatComicMouth
graph {nodes [
  n1 = GetMaterialAttributes(None=FunctionInput() {inputName=Material, inputType=FunctionInput_MaterialAttributes}) {attributeGetTypes [#1, #2]}
  n2 = TextureObjectParameter() {
   group=Mouth Flipbook
   parameterName=Flipbook
   texture=/Game/Characters/Player/Male/Medium/Bodies/M_MED_Buff_Cat_Comic/Textures/BuffCat_Comic_Faces
  }
  n3 = ScalarParameter() {group=Mouth Flipbook, parameterName=UV Offset X}
  n4 = ScalarParameter() {group=Mouth Flipbook, parameterName=UV Offset Y}
  n5 = MaterialFunctionCall(
   UVs=Add(A=TextureCoordinate(), B=AppendVector(A=@n3, B=@n4))
   Texture Scale=ScalarParameter() {defaultValue=1, group=Mouth Flipbook, parameterName=UV Scale}
  ) {materialFunction=&ScaleUVsByCenter}
  n6 = MaterialFunctionCall!(Texture=@n2, UVs=@n5.UVs, SubImages=ScalarParameter() {defaultValue=4, parameterName=Sub Images Count}, Frame=Constant() {r=3}) {materialFunction=/Game/Weapons/WeaponSkins/MaterialFunction/CustomNumbers_SubUV}
  n7 = Multiply(A=@n6.Alpha, B=@n5.0-1 mask)
  FunctionOutput(None=SetMaterialAttributes!(
    MaterialAttributes=@n1.MaterialAttributes
    BaseColor=LinearInterpolate(A=@n1.BaseColor, B=@n6.RGB, Alpha=@n7)
    EmissiveColor=LinearInterpolate(A=@n1.EmissiveColor, B=@n6.RGB, Alpha=@n7)
   ) {attributeSetTypes [#1, #2]})
 ]}
