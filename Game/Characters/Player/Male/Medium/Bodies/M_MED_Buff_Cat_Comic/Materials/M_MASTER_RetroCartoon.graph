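# Master cloth-shaded material for the Buff Cat comic player character, producing a retro cartoon look with ink lines.
# Character PreFX feeds anime shading with pre-skinned bicep line masks and optional flipbook ink, then switchable Banana and sub-UV layers.
# Attribute overrides add distance- and FOV-scaled view-angle highlights, mobile extra lines and optional Fresnel lines before PostFX.
# Face-pose frame selection, mouth and mobile toon-outline nodes are present but their outputs are unused.
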Material /Game/Characters/Player/Male/Medium/Bodies/M_MED_Buff_Cat_Comic/Materials/M_MASTER_RetroCartoon
graph {
 properties {
  bAutomaticallySetUsageInEditor=false
  bUseMaterialAttributes=true
  bUsedWithClothing=true
  bUsedWithMeshDeformer=true
  bUsedWithMorphTargets=true
  bUsedWithSkeletalMesh=true
  pixelDepthOffsetMode=PDOM_Legacy
  shadingModel=MSM_Cloth
  usedShadingModels=Cloth
 }
 nodes [
  n1 = ScalarParameter() {defaultValue=0.1, group=Banana, parameterName=THreshold}
  n2 = ScalarParameter() {group=Banana, parameterName=R_Frown_pose}
  n3 = ScalarParameter() {defaultValue=10, parameterName=Mouth Open Frown Frame}
  n4 = ScalarParameter() {group=Banana, parameterName=Default Open Mouth}
  n5 = ScalarParameter() {group=Banana, parameterName=R_smile_pose}
  n6 = ScalarParameter() {defaultValue=6, parameterName=Mouth Open Smile Frame}
  n7 = ScalarParameter() {defaultValue=1, group=Banana, parameterName=Default Face}
  n8 = If(
   A=Max(A=ScalarParameter() {group=Banana, parameterName=R_lip_corner_narrow_pose}, B=ScalarParameter() {parameterName=PeelyBlenderFace})
   B=@n1
   "A > B"=ScalarParameter() {defaultValue=3, parameterName=Narrow Frame}
   "A == B"=@n7
   "A < B"=@n7
  )
  n9 = If(
   A=ScalarParameter() {group=Banana, parameterName=R_frown_pose}
   B=@n1
   "A > B"=ScalarParameter() {defaultValue=4, parameterName=Frown Frame}
   "A == B"=@n8
   "A < B"=@n8
  )
  n10 = If(
   A=ScalarParameter() {group=Banana, parameterName=R_smile_pose}
   B=@n1
   "A > B"=ScalarParameter() {defaultValue=8, parameterName=Smile Mouth animation Frame}
   "A == B"=@n9
   "A < B"=@n9
  )
  If(
   A=ScalarParameter() {group=Banana, parameterName=Jaw_open_Pose}
   B=@n1
   "A > B"=Max(A=If(A=@n2, B=@n1, "A > B"=@n3, "A == B"=@n4, "A < B"=@n4), B=If(A=@n5, B=@n1, "A > B"=@n6, "A == B"=@n4, "A < B"=@n4))
   "A == B"=@n10
   "A < B"=@n10
  )
  n11 = OneMinus(None=Fresnel(ExponentIn=ScalarParameter() {defaultValue=6, group=HoloFoil, parameterName=FresnelOutlineExponent1}, Normal=PixelNormalWS()) {baseReflectFraction=0})
  n12 = Multiply(
   A=ScalarParameter() {defaultValue=12, group=HoloFoil, parameterName=FresnelOutlineExponent2}
   B=Multiply(A=PixelDepth().R, B=ScalarParameter() {defaultValue=0.01, parameterName=FresnelOutlineOutlineExpMult}) {constB=0.002}
  )
  n13 = Saturate(None=LinearInterpolate(
    A=ScalarParameter() {defaultValue=-4, group=HoloFoil, parameterName=FresnelOutline1}
    B=ScalarParameter() {defaultValue=12, group=HoloFoil, parameterName=FresnelOutline2}
    Alpha=Power(Base=@n11, Exp=@n12)
   ))
  LinearInterpolate(A=Constant() {r=1}, B=@n13)
  n14 = MaterialFunctionCall() {materialFunction=&BaseFN_CharacterFunction_PreFX}
  n15 = PreSkinnedPosition()
  n16 = Abs(None=ComponentMask(None=VertexInterpolator(VS=@n15).PS) {b=true, r=true})
  n17 = Constant2Vector() {g=123, r=30}
  n18 = ScalarParameter() {parameterName=R_Biceps_Lines}
  n19 = ScalarParameter() {parameterName=L_Biceps_Lines}
  n20 = VertexInterpolator(VS=PreSkinnedPosition())
  n21 = Saturate(None=Sign(None=ComponentMask(None=@n20.PS) {r=true}))
  n22 = SphereMask(
   A=@n16
   B=@n17
   Radius=LinearInterpolate(Alpha=LinearInterpolate(A=@n18, B=@n19, Alpha=@n21)) {constA=10, constB=0}
   Hardness=Constant() {r=0.5}
  )
  n23 = Add(
   A=LinearInterpolate(
    A=TextureCoordinate()
    B=TextureCoordinate() {coordinateIndex=1}
    Alpha=ScalarParameter() {group=Flipbook, parameterName=use Second UV Channel}
   )
   B=AppendVector(
    A=ScalarParameter() {defaultValue=0.125, group=Flipbook, parameterName=Flipbook X}
    B=ScalarParameter() {group=Flipbook, parameterName=Flipbook Y}
   )
  )
  n24 = ScalarParameter() {defaultValue=-0.25, group=Flipbook, parameterName=Bump Height}
  n25 = ScalarParameter() {defaultValue=0.5, group=Flipbook, parameterName=Flipbook scale}
  n26 = ScalarParameter() {defaultValue=1, group=Flipbook, parameterName=Flipbook scale V relative}
  n27 = MaterialFunctionCall(
   UVs=LinearInterpolate(
    A=@n23
    B=BumpOffset(Coordinate=@n23, Height=@n24)
    Alpha=ScalarParameter() {defaultValue=1, group=Flipbook, parameterName=BumpOffset Intensity}
   )
   Texture Scale=AppendVector(A=@n25, B=Multiply(A=@n25, B=@n26))
  ) {materialFunction=&ScaleUVsByCenter}
  n28 = ComponentMask(None=MaterialFunctionCall!(
    Texture=TextureObject() {texture=/Game/Characters/Player/Male/Medium/Bodies/M_MED_Buff_Cat_Comic/Textures/BuffCat_Comic_Faces_DF}
    UVs=@n27.UVs
    SubImages=Constant() {r=4}
    Frame=Constant() {r=4}
   ) {materialFunction=/Game/Weapons/WeaponSkins/MaterialFunction/CustomNumbers_SubUV}.RGB) {r=true}
  n29 = Constant() {r=12}
  n30 = MaterialFunctionCall!(
   Material=@n14.Result
   LitDiffuseOverride
   SSC_AlphaOverride
   Tint Mask
   DF InkLine Visibility Mask=StaticSwitchParameter(True=OneMinus(None=@n22), False=Constant() {r=1}) {defaultValue=true, group=Skin FX, parameterName=Biceps Animation}
   DF Inkline Injection=StaticSwitchParameter(True=Multiply(A=@n28, B=@n27.0-1 mask), False=Constant()) {group=Skin FX, parameterName=DF Flipbook}
   DF Inkline Color Override=Constant3Vector() {constant=<rgba 1 1 1 0>}
   DF Inkline Color Override Mask=StaticSwitchParameter(True=SphereMask(A=@n16, B=@n17, Radius=@n29), False=Constant()) {defaultValue=true, group=Skin FX, parameterName=Biceps Animation}
   CustomUVs
  ) {materialFunction=/Game/Packages/Fortress_SharedMaterials/Anime_Master_Material/MF_AnimeShading}
  MaterialFunctionCall!(Material=@n30.Result) {materialFunction=/Game/Packages/Fortress_SharedMaterials/Anime_Master_Material/MF_ToonOutline_Mobile}
  n31 = StaticSwitchParameter(
   True=MaterialFunctionCall!(Gmap=@n14.Gmap, Material=@n30.Result) {materialFunction=/Game/Weapons/WeaponSkins/MaterialFunction/Banana}.Result
   False=@n30.Result
  ) {group=Skin FX, parameterName=Use Banana}
  MaterialFunctionCall!(Material=@n31) {materialFunction=&/MF_BuffCatComicMouth}
  n32 = Constant() {r=1}
  n33 = GetMaterialAttributes(None=StaticSwitchParameter(
    True=MaterialFunctionCall!(Alpha=@n32, Material=@n31) {materialFunction=/Game/Weapons/WeaponSkins/MaterialFunction/SubUVTextures}.Result
    False=@n31
   ) {group=Skin FX, parameterName=Use Sub UV texture}) {attributeGetTypes [#1, #2]}
  n34 = Constant()
  n35 = TextureSampleParameter2D() {parameterName=Mobile Extra Lines, texture=&T_BlacksRGB}
  n36 = ScalarParameter() {defaultValue=1, parameterName=Enable Distance Highlight}
  n37 = Constant() {r=80}
  n38 = Constant() {r=180}
  n39 = ComponentMask(None=VertexInterpolator(VS=PreSkinnedPosition()).PS) {b=true}
  n40 = ObjectRadius()
  n41 = ScalarParameter() {defaultValue=97, parameterName=ObjectScale}
  n42 = Add(A=PixelDepth().R, B=Multiply(B=Divide(A=@n40, B=@n41)) {constA=-250}) {constA=2000, constB=-250}
  n43 = LinearInterpolate(Alpha=MaterialFunctionCall() {materialFunction=/Engine/Functions/Engine_MaterialFunctions02/Utility/FOV}."FOV ") {constA=10000, constB=2000}
  n44 = LinearInterpolate(
   B=LinearInterpolate(B=ScalarParameter() {defaultValue=0.2, parameterName=Highlight Amount}, Alpha=SmoothStep(Min=@n37, Max=@n38, Value=@n39)) {constA=1}
   Alpha=Saturate(None=Divide(A=@n42, B=@n43) {constB=1000})
  ) {constA=2, constAlpha=1, constB=0}
  n45 = ScalarParameter() {defaultValue=0.02, parameterName=Highlight Blur}
  n46 = PixelNormalWS()
  n47 = MaterialFunctionCall() {materialFunction=&CameraDirectionVector}
  n48 = SmoothStep(Min=Subtract(A=@n44, B=@n45), Max=Add(A=@n44, B=@n45), Value=Add(A=DotProduct(A=@n46, B=@n47.Result)))
  n49 = Add(
   A=Max(A=@n33.Emissive Color, B=ShadingPathSwitch(Default=@n34, Mobile=@n35.RGB) {desc=Migrated from FeatureLevelSwitch})
   B=Multiply(A=Multiply(A=@n36, B=@n48)) {constB=0.25}
  )
  n50 = ShadingPathSwitch(Default=Constant() {r=1}, Mobile=@n13) {desc=Migrated from FeatureLevelSwitch}
  n51 = Constant() {r=0.001}
  n52 = Multiply(A=@n51, B=TextureSampleParameter2D() {parameterName=SSC_Texture, texture=&T_BlacksRGB}.B)
  n53 = TextureSampleParameter2D() {parameterName=LitDiffuse, samplerSource=SSM_Wrap_WorldGroupSettings, texture=&T_WhitesRGB}
  n54 = SetMaterialAttributes!(
   MaterialAttributes=@n33.MaterialAttributes
   World Position Offset=@n33.MaterialAttributes
   Emissive Color=StaticSwitchParameter(True=Multiply(A=@n49, B=@n50), False=@n49) {parameterName=Use Fresnel Lines}
   Fuzz Color=StaticSwitchParameter(
    True=VectorParameter() {parameterName=SubsurfaceColor, defaultValue=<rgba 0 0.001 0.001 1>}.RGB
    False=MaterialFunctionCall(X=@n51, Y=@n52) {materialFunction=&MakeFloat3}.Result
   ) {parameterName=UseCustomSubsurfaceColor}
   Opacity=Constant() {r=1}
   Metallic=OneMinus(None=Saturate(None=@n53.R))
  ) {attributeSetTypes [
    #1
    #2
    #3
    #4
    #5
   ]}
  MaterialFunctionCall(AO=@n14.AO, Distance Blend=@n14.AO, Fresnel=@n14.AO, Gmap=@n14.AO, Pre Skinned Local Position=@n14.AO, Material=@n54) {materialFunction=&BaseFN_CharacterFunction_PostFX}
  Max(A=Max(A=@n8, B=@n9), B=@n10)
 ]
}
