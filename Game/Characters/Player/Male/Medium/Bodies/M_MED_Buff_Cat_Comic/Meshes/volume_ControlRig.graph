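# Control Rig for the Buff Cat Comic player body that drives volume-corrective bones at elbows, knees, hips, shoulders, pecs and biceps.
# Weighted parent constraints place each base bone, then joint-rotation variables (remapped and clamped) or psd_arm alphas weight offset and scale corrections.
# Pecs also aim at the upper arms, but the right targets pec_volume_base_r, unlike the left's pec_volume_aim_l.
# TransformConstraintPerItem nodes are unconnected to execution.
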ControlRigBlueprint /Game/Characters/Player/Male/Medium/Bodies/M_MED_Buff_Cat_Comic/Meshes/volume_ControlRig
graphs [
 RigVMFunctionLibrary {links [], nodes []}
 RigVMModel {nodes [
   RigUnit_BeginExecution = &RigUnit_BeginExecution()
   n1 = /Script/ControlRig.RigUnit_SequenceExecution(ExecuteContext=@RigUnit_BeginExecution.ExecutePin) {is=struct RigVMUnitNode}
   &RigUnit_ParentConstraint(
    AdvancedSettings=(RotationOrderForFilter=XZY)
    Child=(Type=Bone,Name="elbow_out_volume_base_l")
    Parents=((Item=(Type=Bone,Name="upperarm_bendy_c_l"),Weight=0.6),(Item=(Type=Bone,Name="lowerarm_bendy_a_l"),Weight=0.4))
    Weight=1
    ExecutePin=@n1.A
   )
   > &RigUnit_ModifyTransforms(
    ItemToModify=((Item=(Type=Bone,Name="elbow_out_volume_l"),Transform=(Rotation=(X=0,Y=0,Z=0,W=1),Translation=(X=0,Y=-3,Z=0),Scale3D=(X=0.3,Y=1,Z=1))))
    Weight=&RigVMFunction_MathFloatRemap(SourceMaximum=80, bClamp=true, Value=RigVMVariableNode() {variable=elbow_rot_l}.Value).Result
   )
   &RigUnit_ParentConstraint(
    AdvancedSettings=(RotationOrderForFilter=XZY)
    Child=(Type=Bone,Name="elbow_in_volume_base_l")
    Parents=((Item=(Type=Bone,Name="upperarm_bendy_c_l"),Weight=0.6),(Item=(Type=Bone,Name="lowerarm_bendy_a_l"),Weight=0.4))
    Weight=1
    ExecutePin=@n1.B
   )
   > &RigUnit_ModifyTransforms(
    ItemToModify=((Item=(Type=Bone,Name="elbow_in_volume_l"),Transform=(Rotation=(X=0,Y=0,Z=0,W=1),Translation=(X=0,Y=8,Z=0),Scale3D=(X=0.5,Y=1,Z=1))))
    Weight=&RigVMFunction_MathFloatRemap(SourceMaximum=90, bClamp=true, Value=RigVMVariableNode() {variable=elbow_rot_l}.Value).Result
   )
   n2 = &RigUnit_BeginExecution()
   n3 = &RigUnit_SequenceExecution(ExecuteContext=@n2.ExecutePin)
   &RigUnit_ParentConstraint(
    AdvancedSettings=(RotationOrderForFilter=XZY)
    Child=(Type=Bone,Name="elbow_in_volume_base_r")
    Parents=((Item=(Type=Bone,Name="upperarm_bendy_c_r"),Weight=0.6),(Item=(Type=Bone,Name="lowerarm_bendy_a_r"),Weight=0.4))
    Weight=1
    ExecutePin=@n3.B
   )
   > &RigUnit_ModifyTransforms(
    ItemToModify=((Item=(Type=Bone,Name="elbow_in_volume_r"),Transform=(Rotation=(X=0,Y=0,Z=0,W=1),Translation=(X=0,Y=-8,Z=0),Scale3D=(X=0.5,Y=1,Z=1))))
    Weight=&RigVMFunction_MathFloatRemap(SourceMaximum=90, bClamp=true, Value=RigVMVariableNode() {variable=elbow_rot_r}.Value).Result
   )
   &RigUnit_ParentConstraint(
    AdvancedSettings=(RotationOrderForFilter=XZY)
    Child=(Type=Bone,Name="elbow_out_volume_base_r")
    Parents=((Item=(Type=Bone,Name="upperarm_bendy_c_r"),Weight=0.6),(Item=(Type=Bone,Name="lowerarm_bendy_a_r"),Weight=0.4))
    Weight=1
    ExecutePin=@n3.A
   )
   > &RigUnit_ModifyTransforms(
    ItemToModify=((Item=(Type=Bone,Name="elbow_out_volume_r"),Transform=(Rotation=(X=0,Y=0,Z=0,W=1),Translation=(X=0,Y=3,Z=0),Scale3D=(X=0.3,Y=1,Z=1))))
    Weight=&RigVMFunction_MathFloatRemap(SourceMaximum=80, bClamp=true, Value=RigVMVariableNode() {variable=elbow_rot_r}.Value).Result
   )
   n4 = &RigUnit_BeginExecution()
   n5 = &RigUnit_SequenceExecution(ExecuteContext=@n4.ExecutePin)
   &RigUnit_ParentConstraint(
    AdvancedSettings=(RotationOrderForFilter=XZY)
    Child=(Type=Bone,Name="crouch_volume_base_l")
    Parents=((Item=(Type=Bone,Name="pelvis"),Weight=0.3),(Item=(Type=Bone,Name="thigh_l"),Weight=0.7))
    Weight=1
    ExecutePin=@n5.A
   )
   > &RigUnit_ModifyTransforms(
    ItemToModify=((Item=(Type=Bone,Name="crouch_volume_l"),Transform=(Rotation=(X=0,Y=0,Z=0,W=1),Translation=(X=-3,Y=10,Z=0),Scale3D=(X=0.6,Y=1,Z=1))))
    Weight=&RigVMFunction_MathFloatRemap(SourceMaximum=0, SourceMinimum=-90, TargetMaximum=0, TargetMinimum=1, bClamp=true, Value=RigVMVariableNode() {variable=thigh_rot_l}.Value).Result
   )
   &RigUnit_ParentConstraint(
    AdvancedSettings=(RotationOrderForFilter=XZY)
    Child=(Type=Bone,Name="glute_volume_base_l")
    Parents=((Item=(Type=Bone,Name="pelvis"),Weight=0.6),(Item=(Type=Bone,Name="thigh_l"),Weight=0.4))
    Weight=1
    ExecutePin=@n5.B
   )
   n6 = RigVMVariableNode() {variable=thigh_rot_l}
   > &RigUnit_ModifyTransforms(
    ItemToModify=((Item=(Type=Bone,Name="glute_volume_l"),Transform=(Rotation=(X=0,Y=0,Z=0,W=1),Translation=(X=-10,Y=-5,Z=0),Scale3D=(X=1,Y=1,Z=1))))
    Weight=&RigVMFunction_MathFloatRemap(SourceMaximum=0, SourceMinimum=-100, TargetMaximum=0, TargetMinimum=1, bClamp=true, Value=@n6.Value).Result
   )
   > &RigUnit_ModifyTransforms(
    ItemToModify=((Item=(Type=Bone,Name="glute_volume_l"),Transform=(Rotation=(X=0,Y=0,Z=0,W=1),Translation=(X=0,Y=-4,Z=0),Scale3D=(X=1,Y=1,Z=1))))
    Weight=&RigVMFunction_MathFloatRemap(SourceMaximum=50, SourceMinimum=0, TargetMinimum=0, bClamp=true, Value=@n6.Value).Result
   )
   n7 = &RigUnit_BeginExecution()
   n8 = &RigUnit_SequenceExecution(ExecuteContext=@n7.ExecutePin)
   &RigUnit_ParentConstraint(
    AdvancedSettings=(RotationOrderForFilter=XZY)
    Child=(Type=Bone,Name="glute_volume_base_r")
    Parents=((Item=(Type=Bone,Name="pelvis"),Weight=0.6),(Item=(Type=Bone,Name="thigh_r"),Weight=0.4))
    Weight=1
    ExecutePin=@n8.B
   )
   n9 = RigVMVariableNode() {variable=thigh_rot_r}
   > &RigUnit_ModifyTransforms(
    ItemToModify=((Item=(Type=Bone,Name="glute_volume_r"),Transform=(Rotation=(X=0,Y=0,Z=0,W=1),Translation=(X=10,Y=5,Z=0),Scale3D=(X=1,Y=1,Z=1))))
    Weight=&RigVMFunction_MathFloatRemap(SourceMaximum=0, SourceMinimum=-100, TargetMaximum=0, TargetMinimum=1, bClamp=true, Value=@n9.Value).Result
   )
   > &RigUnit_ModifyTransforms(
    ItemToModify=((Item=(Type=Bone,Name="glute_volume_r"),Transform=(Rotation=(X=0,Y=0,Z=0,W=1),Translation=(X=0,Y=-4,Z=0),Scale3D=(X=1,Y=1,Z=1))))
    Weight=&RigVMFunction_MathFloatRemap(SourceMaximum=50, SourceMinimum=0, TargetMinimum=0, bClamp=true, Value=@n9.Value).Result
   )
   &RigUnit_ParentConstraint(
    AdvancedSettings=(RotationOrderForFilter=XZY)
    Child=(Type=Bone,Name="crouch_volume_base_r")
    Parents=((Item=(Type=Bone,Name="pelvis"),Weight=0.3),(Item=(Type=Bone,Name="thigh_r"),Weight=0.7))
    Weight=1
    ExecutePin=@n8.A
   )
   > &RigUnit_ModifyTransforms(
    ItemToModify=((Item=(Type=Bone,Name="crouch_volume_r"),Transform=(Rotation=(X=0,Y=0,Z=0,W=1),Translation=(X=3,Y=-10,Z=0),Scale3D=(X=0.6,Y=1,Z=1))))
    Weight=&RigVMFunction_MathFloatRemap(SourceMaximum=0, SourceMinimum=-90, TargetMaximum=0, TargetMinimum=1, bClamp=true, Value=RigVMVariableNode() {variable=thigh_rot_r}.Value).Result
   )
   n10 = &RigUnit_BeginExecution()
   n11 = &RigUnit_SequenceExecution(ExecuteContext=@n10.ExecutePin)
   &RigUnit_ParentConstraint(
    AdvancedSettings=(RotationOrderForFilter=XZY)
    Child=(Type=Bone,Name="knee_in_volume_base_l")
    Parents=((Item=(Type=Bone,Name="upperleg_bendy_c_l"),Weight=0.6),(Item=(Type=Bone,Name="lowerleg_bendy_a_l"),Weight=0.4))
    Weight=1
    ExecutePin=@n11.B
   )
   > &RigUnit_ModifyTransforms(
    ItemToModify=((Item=(Type=Bone,Name="knee_in_volume_l"),Transform=(Rotation=(X=0,Y=0,Z=0,W=1),Translation=(X=0,Y=-10,Z=0),Scale3D=(X=0.5,Y=1,Z=1))))
    Weight=&RigVMFunction_MathFloatRemap(SourceMaximum=100, bClamp=true, Value=RigVMVariableNode() {variable=knee_rot_l}.Value).Result
   )
   &RigUnit_ParentConstraint(
    AdvancedSettings=(RotationOrderForFilter=XZY)
    Child=(Type=Bone,Name="knee_out_volume_base_l")
    Parents=((Item=(Type=Bone,Name="upperleg_bendy_c_l"),Weight=0.5),(Item=(Type=Bone,Name="lowerleg_bendy_a_l"),Weight=0.5))
    Weight=1
    ExecutePin=@n11.A
   )
   > &RigUnit_ModifyTransforms(
    ItemToModify=((Item=(Type=Bone,Name="knee_out_volume_l"),Transform=(Rotation=(X=0,Y=0,Z=0,W=1),Translation=(X=3,Y=3,Z=0),Scale3D=(X=1.5,Y=1,Z=1.5))))
    Weight=&RigVMFunction_MathFloatRemap(SourceMaximum=80, bClamp=true, Value=RigVMVariableNode() {variable=knee_rot_l}.Value).Result
   )
   n12 = &RigUnit_BeginExecution()
   n13 = &RigUnit_SequenceExecution(ExecuteContext=@n12.ExecutePin)
   &RigUnit_ParentConstraint(
    AdvancedSettings=(RotationOrderForFilter=XZY)
    Child=(Type=Bone,Name="knee_in_volume_base_r")
    Parents=((Item=(Type=Bone,Name="upperleg_bendy_c_r"),Weight=0.6),(Item=(Type=Bone,Name="lowerleg_bendy_a_r"),Weight=0.4))
    Weight=1
    ExecutePin=@n13.B
   )
   > &RigUnit_ModifyTransforms(
    ItemToModify=((Item=(Type=Bone,Name="knee_in_volume_r"),Transform=(Rotation=(X=0,Y=0,Z=0,W=1),Translation=(X=0,Y=10,Z=0),Scale3D=(X=0.5,Y=1,Z=1))))
    Weight=&RigVMFunction_MathFloatRemap(SourceMaximum=100, bClamp=true, Value=RigVMVariableNode() {variable=knee_rot_r}.Value).Result
   )
   &RigUnit_ParentConstraint(
    AdvancedSettings=(RotationOrderForFilter=XZY)
    Child=(Type=Bone,Name="knee_out_volume_base_r")
    Parents=((Item=(Type=Bone,Name="upperleg_bendy_c_r"),Weight=0.5),(Item=(Type=Bone,Name="lowerleg_bendy_a_r"),Weight=0.5))
    Weight=1
    ExecutePin=@n13.A
   )
   > &RigUnit_ModifyTransforms(
    ItemToModify=((Item=(Type=Bone,Name="knee_out_volume_r"),Transform=(Rotation=(X=0,Y=0,Z=0,W=1),Translation=(X=-3,Y=-3,Z=0),Scale3D=(X=1.5,Y=1,Z=1.5))))
    Weight=&RigVMFunction_MathFloatRemap(SourceMaximum=80, bClamp=true, Value=RigVMVariableNode() {variable=knee_rot_r}.Value).Result
   )
   n14 = &RigUnit_BeginExecution()
   n15 = &RigUnit_SequenceExecution(ExecuteContext=@n14.ExecutePin)
   &RigUnit_ParentConstraint(
    AdvancedSettings=(RotationOrderForFilter=XZY)
    Child=(Type=Bone,Name="pec_volume_aim_base_l")
    Parents=((Item=(Type=Bone,Name="spine_03"),Weight=0.25),(Item=(Type=Bone,Name="spine_04"),Weight=0.55),(Item=(Type=Bone,Name="spine_05"),Weight=0.2))
    Weight=1
    ExecutePin=@n15.A
   )
   > /Script/ControlRig.RigUnit_AimItem(
    DebugSettings=(Scale=10,WorldOffset=(Rotation=(X=0,Y=0,Z=0,W=1),Translation=(X=0,Y=0,Z=0),Scale3D=(X=1,Y=1,Z=1)))
    Item=(Type=Bone,Name="pec_volume_aim_l")
    Primary=(Weight=1,Axis=(X=1,Y=0,Z=0),Target=(X=0,Y=0,Z=0),Kind=Location,Space=(Type=Bone,Name="upperarm_l"))
    Secondary=(Weight=0,Axis=(X=0,Y=0,Z=1),Target=(X=1,Y=0,Z=0),Kind=Location,Space=(Type=Bone))
    Weight=1
   ) {is=struct RigVMUnitNode}
   n16 = &RigUnit_BeginExecution()
   n17 = &RigUnit_SequenceExecution(ExecuteContext=@n16.ExecutePin)
   &RigUnit_ParentConstraint(
    AdvancedSettings=(RotationOrderForFilter=XZY)
    Child=(Type=Bone,Name="pec_volume_aim_base_r")
    Parents=((Item=(Type=Bone,Name="spine_03"),Weight=0.25),(Item=(Type=Bone,Name="spine_04"),Weight=0.55),(Item=(Type=Bone,Name="spine_05"),Weight=0.2))
    Weight=1
    ExecutePin=@n17.A
   )
   > &RigUnit_AimItem(
    DebugSettings=(Scale=10,WorldOffset=(Rotation=(X=0,Y=0,Z=0,W=1),Translation=(X=0,Y=0,Z=0),Scale3D=(X=1,Y=1,Z=1)))
    Item=(Type=Bone,Name="pec_volume_base_r")
    Primary=(Weight=1,Axis=(X=-1,Y=0,Z=0),Target=(X=0,Y=0,Z=0),Kind=Location,Space=(Type=Bone,Name="upperarm_r"))
    Secondary=(Weight=0,Axis=(X=0,Y=0,Z=1),Target=(X=1,Y=0,Z=0),Kind=Location,Space=(Type=Bone))
    Weight=1
   )
   n18 = &RigUnit_BeginExecution()
   n19 = &RigUnit_SequenceExecution(ExecuteContext=@n18.ExecutePin)
   &RigUnit_ParentConstraint(
    AdvancedSettings=(RotationOrderForFilter=XZY)
    Child=(Type=Bone,Name="upperarm_up_volume_base_l")
    Parents=((Item=(Type=Bone,Name="upperarm_twist_01_l"),Weight=0.6),(Item=(Type=Bone,Name="clavicle_l"),Weight=0.4))
    Weight=1
    ExecutePin=@n19.A
   )
   > &RigUnit_ModifyTransforms(
    ItemToModify=((Item=(Type=Bone,Name="upperarm_up_volume_l"),Transform=(Rotation=(X=0,Y=0,Z=0,W=1),Translation=(X=0,Y=0,Z=4),Scale3D=(X=1,Y=1,Z=1))))
    Weight=RigVMVariableNode() {variable=psd_arm_up_alpha_l}.Value
   )
   &RigUnit_ParentConstraint(
    AdvancedSettings=(RotationOrderForFilter=XZY)
    Child=(Type=Bone,Name="upperarm_fr_volume_base_l")
    Parents=((Item=(Type=Bone,Name="upperarm_twist_01_l"),Weight=0.3),(Item=(Type=Bone,Name="clavicle_l"),Weight=0.7))
    Weight=1
    ExecutePin=@n19.B
   )
   n20 = RigVMVariableNode() {variable=psd_arm_fwd_alpha_l}
   > &RigUnit_ModifyTransforms(
    ItemToModify=((Item=(Type=Bone,Name="upperarm_fr_volume_l"),Transform=(Rotation=(X=0,Y=0,Z=0,W=1),Translation=(X=0,Y=10,Z=0),Scale3D=(X=1,Y=1,Z=1))))
    Weight=@n20.Value
   )
   > &RigUnit_ModifyTransforms(
    ItemToModify=((Item=(Type=Bone,Name="upperarm_bk_volume_l"),Transform=(Rotation=(X=0,Y=0,Z=0,W=1),Translation=(X=0,Y=-6,Z=0),Scale3D=(X=1,Y=1,Z=1))))
    Weight=@n20.Value
   )
   > &RigUnit_ModifyTransforms(
    ItemToModify=((Item=(Type=Bone,Name="pec_volume_l"),Transform=(Rotation=(X=0,Y=0,Z=0,W=1),Translation=(X=-3,Y=0,Z=0),Scale3D=(X=1,Y=1,Z=1))))
    Weight=RigVMVariableNode() {variable=psd_arm_fwd_dn_alpha_l}.Value
   )
   &RigUnit_ParentConstraint(
    AdvancedSettings=(RotationOrderForFilter=XZY)
    Child=(Type=Bone,Name="upperarm_bk_volume_base_l")
    Parents=((Item=(Type=Bone,Name="upperarm_twist_01_l"),Weight=0.4),(Item=(Type=Bone,Name="clavicle_l"),Weight=0.6))
    Weight=1
    ExecutePin=@n19.C
   )
   > &RigUnit_ModifyTransforms(
    ItemToModify=((Item=(Type=Bone,Name="upperarm_bk_volume_l"),Transform=(Rotation=(X=0,Y=0,Z=0,W=1),Translation=(X=0,Y=-4,Z=0),Scale3D=(X=1,Y=1,Z=1))))
    Weight=RigVMVariableNode() {variable=psd_arm_bk_alpha_l}.Value
   )
   n21 = &RigUnit_BeginExecution()
   n22 = &RigUnit_SequenceExecution(ExecuteContext=@n21.ExecutePin)
   &RigUnit_ParentConstraint(
    AdvancedSettings=(RotationOrderForFilter=XZY)
    Child=(Type=Bone,Name="upperarm_bk_volume_base_r")
    Parents=((Item=(Type=Bone,Name="upperarm_twist_01_r"),Weight=0.4),(Item=(Type=Bone,Name="clavicle_r"),Weight=0.6))
    Weight=1
    ExecutePin=@n22.C
   )
   > &RigUnit_ModifyTransforms(
    ItemToModify=((Item=(Type=Bone,Name="upperarm_bk_volume_r"),Transform=(Rotation=(X=0,Y=0,Z=0,W=1),Translation=(X=0,Y=4,Z=0),Scale3D=(X=1,Y=1,Z=1))))
    Weight=RigVMVariableNode() {variable=psd_arm_bk_alpha_r}.Value
   )
   &RigUnit_ParentConstraint(
    AdvancedSettings=(RotationOrderForFilter=XZY)
    Child=(Type=Bone,Name="upperarm_fr_volume_base_r")
    Parents=((Item=(Type=Bone,Name="upperarm_twist_01_r"),Weight=0.3),(Item=(Type=Bone,Name="clavicle_r"),Weight=0.7))
    Weight=1
    ExecutePin=@n22.B
   )
   n23 = RigVMVariableNode() {variable=psd_arm_fwd_alpha_r}
   > &RigUnit_ModifyTransforms(
    ItemToModify=((Item=(Type=Bone,Name="upperarm_fr_volume_r"),Transform=(Rotation=(X=0,Y=0,Z=0,W=1),Translation=(X=0,Y=-10,Z=0),Scale3D=(X=1,Y=1,Z=1))))
    Weight=@n23.Value
   )
   > &RigUnit_ModifyTransforms(
    ItemToModify=((Item=(Type=Bone,Name="upperarm_bk_volume_r"),Transform=(Rotation=(X=0,Y=0,Z=0,W=1),Translation=(X=0,Y=6,Z=0),Scale3D=(X=1,Y=1,Z=1))))
    Weight=@n23.Value
   )
   > &RigUnit_ModifyTransforms(
    ItemToModify=((Item=(Type=Bone,Name="pec_volume_r"),Transform=(Rotation=(X=0,Y=0,Z=0,W=1),Translation=(X=3,Y=0,Z=0),Scale3D=(X=1,Y=1,Z=1))))
    Weight=RigVMVariableNode() {variable=psd_arm_fwd_dn_alpha_r}.Value
   )
   &RigUnit_ParentConstraint(
    AdvancedSettings=(RotationOrderForFilter=XZY)
    Child=(Type=Bone,Name="upperarm_up_volume_base_r")
    Parents=((Item=(Type=Bone,Name="upperarm_twist_01_r"),Weight=0.6),(Item=(Type=Bone,Name="clavicle_r"),Weight=0.4))
    Weight=1
    ExecutePin=@n22.A
   )
   > &RigUnit_ModifyTransforms(
    ItemToModify=((Item=(Type=Bone,Name="upperarm_up_volume_r"),Transform=(Rotation=(X=0,Y=0,Z=0,W=1),Translation=(X=0,Y=0,Z=-4),Scale3D=(X=1,Y=1,Z=1))))
    Weight=RigVMVariableNode() {variable=psd_arm_up_alpha_r}.Value
   )
   n24 = &RigUnit_BeginExecution()
   n25 = &RigUnit_SequenceExecution(ExecuteContext=@n24.ExecutePin)
   &RigUnit_ParentConstraint(
    AdvancedSettings=(RotationOrderForFilter=XZY)
    Child=(Type=Bone,Name="bicep_volume_base_l")
    Parents=((Item=(Type=Bone,Name="upperarm_bendy_c_l"),Weight=0.5),(Item=(Type=Bone,Name="upperarm_bendy_b_l"),Weight=0.3),(Item=(Type=Bone,Name="upperarm_twist_01_l"),Weight=0.2))
    Weight=1
    ExecutePin=@n25.A
   )
   n26 = &RigUnit_BeginExecution()
   n27 = &RigUnit_SequenceExecution(ExecuteContext=@n26.ExecutePin)
   &RigUnit_ParentConstraint(
    AdvancedSettings=(RotationOrderForFilter=XZY)
    Child=(Type=Bone,Name="bicep_volume_base_r")
    Parents=((Item=(Type=Bone,Name="upperarm_bendy_c_r"),Weight=0.5),(Item=(Type=Bone,Name="upperarm_bendy_b_r"),Weight=0.3),(Item=(Type=Bone,Name="upperarm_twist_01_r"),Weight=0.2))
    Weight=1
    ExecutePin=@n27.A
   )
   /Script/ControlRig.RigUnit_TransformConstraintPerItem(
    BaseItem=(Type=Bone)
    BaseTransform=(Rotation=(X=0,Y=0,Z=0,W=1),Translation=(X=0,Y=0,Z=0),Scale3D=(X=1,Y=1,Z=1))
    BaseTransformSpace=GlobalSpace
    ExecutePin=()
    Item=(Type=Bone,Name="elbow_out_volume_base_l")
    Targets="((Transform=(Rotation=(X=0,Y=0,Z=0,W=1),Translation=(X=0,Y=0,Z=0),Scale3D=(X=1,Y=1,Z=1)),Weight=0.6,bMaintainOffset=True,Filter=(TranslationFilter=(bX=True,bY=True,bZ=True),RotationFilter=(bX=True,bY=True,bZ=True),ScaleFilter=(bX=True,bY=True,bZ=True))),(T ...(truncated)"
    bUseInitialTransforms=True
    Targets.0.Transform=&RigUnit_GetTransform(Item=(Type=Bone,Name="upperarm_bendy_c_l")).Transform
    Targets.1.Transform=&RigUnit_GetTransform(Item=(Type=Bone,Name="lowerarm_bendy_a_l")).Transform
   ) {is=struct RigVMUnitNode}
   &RigUnit_TransformConstraintPerItem(
    BaseItem=(Type=Bone)
    BaseTransform=(Rotation=(X=0,Y=0,Z=0,W=1),Translation=(X=0,Y=0,Z=0),Scale3D=(X=1,Y=1,Z=1))
    BaseTransformSpace=GlobalSpace
    ExecutePin=()
    Item=(Type=Bone,Name="elbow_in_volume_base_l")
    Targets="((Transform=(Rotation=(X=0,Y=0,Z=0,W=1),Translation=(X=0,Y=0,Z=0),Scale3D=(X=1,Y=1,Z=1)),Weight=0.6,bMaintainOffset=True,Filter=(TranslationFilter=(bX=True,bY=True,bZ=True),RotationFilter=(bX=True,bY=True,bZ=True),ScaleFilter=(bX=True,bY=True,bZ=True))),(T ...(truncated)"
    bUseInitialTransforms=True
    Targets.1.Transform=&RigUnit_GetTransform(Item=(Type=Bone,Name="lowerarm_bendy_a_l")).Transform
    Targets.0.Transform=&RigUnit_GetTransform(Item=(Type=Bone,Name="upperarm_bendy_c_l")).Transform
   )
   &RigUnit_TransformConstraintPerItem(
    BaseItem=(Type=Bone)
    BaseTransform=(Rotation=(X=0,Y=0,Z=0,W=1),Translation=(X=0,Y=0,Z=0),Scale3D=(X=1,Y=1,Z=1))
    BaseTransformSpace=GlobalSpace
    ExecutePin=()
    Item=(Type=Bone,Name="elbow_in_volume_base_r")
    Targets="((Transform=(Rotation=(X=0,Y=0,Z=0,W=1),Translation=(X=0,Y=0,Z=0),Scale3D=(X=1,Y=1,Z=1)),Weight=0.6,bMaintainOffset=True,Filter=(TranslationFilter=(bX=True,bY=True,bZ=True),RotationFilter=(bX=True,bY=True,bZ=True),ScaleFilter=(bX=True,bY=True,bZ=True))),(T ...(truncated)"
    bUseInitialTransforms=True
    Targets.1.Transform=&RigUnit_GetTransform(Item=(Type=Bone,Name="lowerarm_bendy_a_r")).Transform
    Targets.0.Transform=&RigUnit_GetTransform(Item=(Type=Bone,Name="upperarm_bendy_c_r")).Transform
   )
   &RigUnit_TransformConstraintPerItem(
    BaseItem=(Type=Bone)
    BaseTransform=(Rotation=(X=0,Y=0,Z=0,W=1),Translation=(X=0,Y=0,Z=0),Scale3D=(X=1,Y=1,Z=1))
    BaseTransformSpace=GlobalSpace
    ExecutePin=()
    Item=(Type=Bone,Name="crouch_volume_base_l")
    Targets="((Transform=(Rotation=(X=0,Y=0,Z=0,W=1),Translation=(X=0,Y=0,Z=0),Scale3D=(X=1,Y=1,Z=1)),Weight=0.3,bMaintainOffset=True,Filter=(TranslationFilter=(bX=True,bY=True,bZ=True),RotationFilter=(bX=True,bY=True,bZ=True),ScaleFilter=(bX=True,bY=True,bZ=True))),(T ...(truncated)"
    bUseInitialTransforms=True
    Targets.1.Transform=&RigUnit_GetTransform(Item=(Type=Bone,Name="thigh_l")).Transform
    Targets.0.Transform=&RigUnit_GetTransform(Item=(Type=Bone,Name="pelvis")).Transform
   )
   &RigUnit_TransformConstraintPerItem(
    BaseItem=(Type=Bone)
    BaseTransform=(Rotation=(X=0,Y=0,Z=0,W=1),Translation=(X=0,Y=0,Z=0),Scale3D=(X=1,Y=1,Z=1))
    BaseTransformSpace=GlobalSpace
    ExecutePin=()
    Item=(Type=Bone,Name="glute_volume_base_l")
    Targets="((Transform=(Rotation=(X=0,Y=0,Z=0,W=1),Translation=(X=0,Y=0,Z=0),Scale3D=(X=1,Y=1,Z=1)),Weight=0.6,bMaintainOffset=True,Filter=(TranslationFilter=(bX=True,bY=True,bZ=True),RotationFilter=(bX=True,bY=True,bZ=True),ScaleFilter=(bX=True,bY=True,bZ=True))),(T ...(truncated)"
    bUseInitialTransforms=True
    Targets.1.Transform=&RigUnit_GetTransform(Item=(Type=Bone,Name="thigh_l"), bInitial=false).Transform
    Targets.0.Transform=&RigUnit_GetTransform(Item=(Type=Bone,Name="pelvis"), bInitial=false).Transform
   )
   &RigUnit_TransformConstraintPerItem(
    BaseItem=(Type=Bone)
    BaseTransform=(Rotation=(X=0,Y=0,Z=0,W=1),Translation=(X=0,Y=0,Z=0),Scale3D=(X=1,Y=1,Z=1))
    BaseTransformSpace=GlobalSpace
    ExecutePin=()
    Item=(Type=Bone,Name="glute_volume_base_r")
    Targets="((Transform=(Rotation=(X=0,Y=0,Z=0,W=1),Translation=(X=0,Y=0,Z=0),Scale3D=(X=1,Y=1,Z=1)),Weight=0.6,bMaintainOffset=True,Filter=(TranslationFilter=(bX=True,bY=True,bZ=True),RotationFilter=(bX=True,bY=True,bZ=True),ScaleFilter=(bX=True,bY=True,bZ=True))),(T ...(truncated)"
    bUseInitialTransforms=True
    Targets.1.Transform=&RigUnit_GetTransform(Item=(Type=Bone,Name="thigh_r")).Transform
    Targets.0.Transform=&RigUnit_GetTransform(Item=(Type=Bone,Name="pelvis")).Transform
   )
   &RigUnit_TransformConstraintPerItem(
    BaseItem=(Type=Bone)
    BaseTransform=(Rotation=(X=0,Y=0,Z=0,W=1),Translation=(X=0,Y=0,Z=0),Scale3D=(X=1,Y=1,Z=1))
    BaseTransformSpace=GlobalSpace
    ExecutePin=()
    Item=(Type=Bone,Name="crouch_volume_base_r")
    Targets="((Transform=(Rotation=(X=0,Y=0,Z=0,W=1),Translation=(X=0,Y=0,Z=0),Scale3D=(X=1,Y=1,Z=1)),Weight=0.3,bMaintainOffset=True,Filter=(TranslationFilter=(bX=True,bY=True,bZ=True),RotationFilter=(bX=True,bY=True,bZ=True),ScaleFilter=(bX=True,bY=True,bZ=True))),(T ...(truncated)"
    bUseInitialTransforms=True
    Targets.1.Transform=&RigUnit_GetTransform(Item=(Type=Bone,Name="thigh_r")).Transform
    Targets.0.Transform=&RigUnit_GetTransform(Item=(Type=Bone,Name="pelvis")).Transform
   )
   &RigUnit_TransformConstraintPerItem(
    BaseItem=(Type=Bone)
    BaseTransform=(Rotation=(X=0,Y=0,Z=0,W=1),Translation=(X=0,Y=0,Z=0),Scale3D=(X=1,Y=1,Z=1))
    BaseTransformSpace=GlobalSpace
    ExecutePin=()
    Item=(Type=Bone,Name="knee_in_volume_base_l")
    Targets="((Transform=(Rotation=(X=0,Y=0,Z=0,W=1),Translation=(X=0,Y=0,Z=0),Scale3D=(X=1,Y=1,Z=1)),Weight=0.6,bMaintainOffset=True,Filter=(TranslationFilter=(bX=True,bY=True,bZ=True),RotationFilter=(bX=True,bY=True,bZ=True),ScaleFilter=(bX=True,bY=True,bZ=True))),(T ...(truncated)"
    bUseInitialTransforms=True
    Targets.1.Transform=&RigUnit_GetTransform(Item=(Type=Bone,Name="lowerleg_bendy_a_l")).Transform
    Targets.0.Transform=&RigUnit_GetTransform(Item=(Type=Bone,Name="upperleg_bendy_c_l")).Transform
   )
   &RigUnit_TransformConstraintPerItem(
    BaseItem=(Type=Bone)
    BaseTransform=(Rotation=(X=0,Y=0,Z=0,W=1),Translation=(X=0,Y=0,Z=0),Scale3D=(X=1,Y=1,Z=1))
    BaseTransformSpace=GlobalSpace
    ExecutePin=()
    Item=(Type=Bone,Name="knee_in_volume_base_r")
    Targets="((Transform=(Rotation=(X=0,Y=0,Z=0,W=1),Translation=(X=0,Y=0,Z=0),Scale3D=(X=1,Y=1,Z=1)),Weight=0.6,bMaintainOffset=True,Filter=(TranslationFilter=(bX=True,bY=True,bZ=True),RotationFilter=(bX=True,bY=True,bZ=True),ScaleFilter=(bX=True,bY=True,bZ=True))),(T ...(truncated)"
    bUseInitialTransforms=True
    Targets.1.Transform=&RigUnit_GetTransform(Item=(Type=Bone,Name="calf_r")).Transform
    Targets.0.Transform=&RigUnit_GetTransform(Item=(Type=Bone,Name="upperleg_bendy_c_r")).Transform
   )
   &RigUnit_TransformConstraintPerItem(
    BaseItem=(Type=Bone)
    BaseTransform=(Rotation=(X=0,Y=0,Z=0,W=1),Translation=(X=0,Y=0,Z=0),Scale3D=(X=1,Y=1,Z=1))
    BaseTransformSpace=GlobalSpace
    ExecutePin=()
    Item=(Type=Bone,Name="upperarm_up_volume_base_l")
    Targets="((Transform=(Rotation=(X=0,Y=0,Z=0,W=1),Translation=(X=0,Y=0,Z=0),Scale3D=(X=1,Y=1,Z=1)),Weight=0.6,bMaintainOffset=True,Filter=(TranslationFilter=(bX=True,bY=True,bZ=True),RotationFilter=(bX=True,bY=True,bZ=True),ScaleFilter=(bX=True,bY=True,bZ=True))),(T ...(truncated)"
    bUseInitialTransforms=True
    Targets.1.Transform=&RigUnit_GetTransform(Item=(Type=Bone,Name="clavicle_l")).Transform
    Targets.0.Transform=&RigUnit_GetTransform(Item=(Type=Bone,Name="upperarm_twist_01_l")).Transform
   )
   &RigUnit_TransformConstraintPerItem(
    BaseItem=(Type=Bone)
    BaseTransform=(Rotation=(X=0,Y=0,Z=0,W=1),Translation=(X=0,Y=0,Z=0),Scale3D=(X=1,Y=1,Z=1))
    BaseTransformSpace=GlobalSpace
    ExecutePin=()
    Item=(Type=Bone,Name="upperarm_fr_volume_base_l")
    Targets="((Transform=(Rotation=(X=0,Y=0,Z=0,W=1),Translation=(X=0,Y=0,Z=0),Scale3D=(X=1,Y=1,Z=1)),Weight=0.3,bMaintainOffset=True,Filter=(TranslationFilter=(bX=True,bY=True,bZ=True),RotationFilter=(bX=True,bY=True,bZ=True),ScaleFilter=(bX=True,bY=True,bZ=True))),(T ...(truncated)"
    bUseInitialTransforms=True
    Targets.1.Transform=&RigUnit_GetTransform(Item=(Type=Bone,Name="clavicle_l")).Transform
    Targets.0.Transform=&RigUnit_GetTransform(Item=(Type=Bone,Name="upperarm_twist_01_l")).Transform
   )
   &RigUnit_TransformConstraintPerItem(
    BaseItem=(Type=Bone)
    BaseTransform=(Rotation=(X=0,Y=0,Z=0,W=1),Translation=(X=0,Y=0,Z=0),Scale3D=(X=1,Y=1,Z=1))
    BaseTransformSpace=GlobalSpace
    ExecutePin=()
    Item=(Type=Bone,Name="upperarm_bk_volume_base_l")
    Targets="((Transform=(Rotation=(X=0,Y=0,Z=0,W=1),Translation=(X=0,Y=0,Z=0),Scale3D=(X=1,Y=1,Z=1)),Weight=0.4,bMaintainOffset=True,Filter=(TranslationFilter=(bX=True,bY=True,bZ=True),RotationFilter=(bX=True,bY=True,bZ=True),ScaleFilter=(bX=True,bY=True,bZ=True))),(T ...(truncated)"
    bUseInitialTransforms=True
    Targets.1.Transform=&RigUnit_GetTransform(Item=(Type=Bone,Name="clavicle_l")).Transform
    Targets.0.Transform=&RigUnit_GetTransform(Item=(Type=Bone,Name="upperarm_twist_01_l")).Transform
   )
   &RigUnit_TransformConstraintPerItem(
    BaseItem=(Type=Bone)
    BaseTransform=(Rotation=(X=0,Y=0,Z=0,W=1),Translation=(X=0,Y=0,Z=0),Scale3D=(X=1,Y=1,Z=1))
    BaseTransformSpace=GlobalSpace
    ExecutePin=()
    Item=(Type=Bone,Name="upperarm_bk_volume_base_r")
    Targets="((Transform=(Rotation=(X=0,Y=0,Z=0,W=1),Translation=(X=0,Y=0,Z=0),Scale3D=(X=1,Y=1,Z=1)),Weight=0.4,bMaintainOffset=True,Filter=(TranslationFilter=(bX=True,bY=True,bZ=True),RotationFilter=(bX=True,bY=True,bZ=True),ScaleFilter=(bX=True,bY=True,bZ=True))),(T ...(truncated)"
    bUseInitialTransforms=True
    Targets.1.Transform=&RigUnit_GetTransform(Item=(Type=Bone,Name="clavicle_r")).Transform
    Targets.0.Transform=&RigUnit_GetTransform(Item=(Type=Bone,Name="upperarm_twist_01_r")).Transform
   )
   &RigUnit_TransformConstraintPerItem(
    BaseItem=(Type=Bone)
    BaseTransform=(Rotation=(X=0,Y=0,Z=0,W=1),Translation=(X=0,Y=0,Z=0),Scale3D=(X=1,Y=1,Z=1))
    BaseTransformSpace=GlobalSpace
    ExecutePin=()
    Item=(Type=Bone,Name="upperarm_fr_volume_base_r")
    Targets="((Transform=(Rotation=(X=0,Y=0,Z=0,W=1),Translation=(X=0,Y=0,Z=0),Scale3D=(X=1,Y=1,Z=1)),Weight=0.3,bMaintainOffset=True,Filter=(TranslationFilter=(bX=True,bY=True,bZ=True),RotationFilter=(bX=True,bY=True,bZ=True),ScaleFilter=(bX=True,bY=True,bZ=True))),(T ...(truncated)"
    bUseInitialTransforms=True
    Targets.1.Transform=&RigUnit_GetTransform(Item=(Type=Bone,Name="clavicle_r")).Transform
    Targets.0.Transform=&RigUnit_GetTransform(Item=(Type=Bone,Name="upperarm_twist_01_r")).Transform
   )
   &RigUnit_TransformConstraintPerItem(
    BaseItem=(Type=Bone)
    BaseTransform=(Rotation=(X=0,Y=0,Z=0,W=1),Translation=(X=0,Y=0,Z=0),Scale3D=(X=1,Y=1,Z=1))
    BaseTransformSpace=GlobalSpace
    ExecutePin=()
    Item=(Type=Bone,Name="upperarm_up_volume_base_r")
    Targets="((Transform=(Rotation=(X=0,Y=0,Z=0,W=1),Translation=(X=0,Y=0,Z=0),Scale3D=(X=1,Y=1,Z=1)),Weight=0.6,bMaintainOffset=True,Filter=(TranslationFilter=(bX=True,bY=True,bZ=True),RotationFilter=(bX=True,bY=True,bZ=True),ScaleFilter=(bX=True,bY=True,bZ=True))),(T ...(truncated)"
    bUseInitialTransforms=True
    Targets.1.Transform=&RigUnit_GetTransform(Item=(Type=Bone,Name="clavicle_r")).Transform
    Targets.0.Transform=&RigUnit_GetTransform(Item=(Type=Bone,Name="upperarm_twist_01_r")).Transform
   )
   &RigUnit_TransformConstraintPerItem(
    BaseItem=(Type=Bone)
    BaseTransform=(Rotation=(X=0,Y=0,Z=0,W=1),Translation=(X=0,Y=0,Z=0),Scale3D=(X=1,Y=1,Z=1))
    BaseTransformSpace=GlobalSpace
    ExecutePin=()
    Item=(Type=Bone,Name="pec_volume_base_l")
    Targets="((Transform=(Rotation=(X=0,Y=0,Z=0,W=1),Translation=(X=0,Y=0,Z=0),Scale3D=(X=1,Y=1,Z=1)),Weight=0.5,bMaintainOffset=True,Filter=(TranslationFilter=(bX=True,bY=True,bZ=True),RotationFilter=(bX=True,bY=True,bZ=True),ScaleFilter=(bX=True,bY=True,bZ=True))),(T ...(truncated)"
    bUseInitialTransforms=True
    Targets.1.Transform.Translation=&RigUnit_GetTransform(Item=(Type=Bone,Name="upperarm_l")).Transform.Translation
    Targets.0.Transform.Translation=&RigUnit_GetTransform(Item=(Type=Bone,Name="pec_volume_aim_l")).Transform.Translation
   )
   &RigUnit_TransformConstraintPerItem(
    BaseItem=(Type=Bone)
    BaseTransform=(Rotation=(X=0,Y=0,Z=0,W=1),Translation=(X=0,Y=0,Z=0),Scale3D=(X=1,Y=1,Z=1))
    BaseTransformSpace=GlobalSpace
    ExecutePin=()
    Item=(Type=Bone,Name="pec_volume_base_r")
    Targets="((Transform=(Rotation=(X=0,Y=0,Z=0,W=1),Translation=(X=0,Y=0,Z=0),Scale3D=(X=1,Y=1,Z=1)),Weight=0.5,bMaintainOffset=True,Filter=(TranslationFilter=(bX=True,bY=True,bZ=True),RotationFilter=(bX=True,bY=True,bZ=True),ScaleFilter=(bX=True,bY=True,bZ=True))),(T ...(truncated)"
    bUseInitialTransforms=True
    Targets.1.Transform.Translation=&RigUnit_GetTransform(Item=(Type=Bone,Name="upperarm_r")).Transform.Translation
    Targets.0.Transform.Translation=&RigUnit_GetTransform(Item=(Type=Bone,Name="pec_volume_aim_r")).Transform.Translation
   )
   &RigUnit_TransformConstraintPerItem(
    BaseItem=(Type=Bone)
    BaseTransform=(Rotation=(X=0,Y=0,Z=0,W=1),Translation=(X=0,Y=0,Z=0),Scale3D=(X=1,Y=1,Z=1))
    BaseTransformSpace=GlobalSpace
    ExecutePin=()
    Item=(Type=Bone,Name="elbow_out_volume_base_r")
    Targets="((Transform=(Rotation=(X=0,Y=0,Z=0,W=1),Translation=(X=0,Y=0,Z=0),Scale3D=(X=1,Y=1,Z=1)),Weight=0.6,bMaintainOffset=True,Filter=(TranslationFilter=(bX=True,bY=True,bZ=True),RotationFilter=(bX=True,bY=True,bZ=True),ScaleFilter=(bX=True,bY=True,bZ=True))),(T ...(truncated)"
    bUseInitialTransforms=True
    Targets.1.Transform=&RigUnit_GetTransform(Item=(Type=Bone,Name="lowerarm_bendy_a_r")).Transform
    Targets.0.Transform=&RigUnit_GetTransform(Item=(Type=Bone,Name="upperarm_bendy_c_r")).Transform
   )
   &RigUnit_TransformConstraintPerItem(
    BaseItem=(Type=Bone)
    BaseTransform=(Rotation=(X=0,Y=0,Z=0,W=1),Translation=(X=0,Y=0,Z=0),Scale3D=(X=1,Y=1,Z=1))
    BaseTransformSpace=GlobalSpace
    ExecutePin=()
    Item=(Type=Bone,Name="knee_out_volume_base_l")
    Targets="((Transform=(Rotation=(X=0,Y=0,Z=0,W=1),Translation=(X=0,Y=0,Z=0),Scale3D=(X=1,Y=1,Z=1)),Weight=0.5,bMaintainOffset=True,Filter=(TranslationFilter=(bX=True,bY=True,bZ=True),RotationFilter=(bX=True,bY=True,bZ=True),ScaleFilter=(bX=True,bY=True,bZ=True))),(T ...(truncated)"
    bUseInitialTransforms=True
    Targets.1.Transform=&RigUnit_GetTransform(Item=(Type=Bone,Name="lowerleg_bendy_a_l")).Transform
    Targets.0.Transform=&RigUnit_GetTransform(Item=(Type=Bone,Name="upperleg_bendy_c_l")).Transform
   )
   &RigUnit_TransformConstraintPerItem(
    BaseItem=(Type=Bone)
    BaseTransform=(Rotation=(X=0,Y=0,Z=0,W=1),Translation=(X=0,Y=0,Z=0),Scale3D=(X=1,Y=1,Z=1))
    BaseTransformSpace=GlobalSpace
    ExecutePin=()
    Item=(Type=Bone,Name="knee_out_volume_base_r")
    Targets="((Transform=(Rotation=(X=0,Y=0,Z=0,W=1),Translation=(X=0,Y=0,Z=0),Scale3D=(X=1,Y=1,Z=1)),Weight=0.5,bMaintainOffset=True,Filter=(TranslationFilter=(bX=True,bY=True,bZ=True),RotationFilter=(bX=True,bY=True,bZ=True),ScaleFilter=(bX=True,bY=True,bZ=True))),(T ...(truncated)"
    bUseInitialTransforms=True
    Targets.1.Transform=&RigUnit_GetTransform(Item=(Type=Bone,Name="lowerleg_bendy_a_r")).Transform
    Targets.0.Transform=&RigUnit_GetTransform(Item=(Type=Bone,Name="upperleg_bendy_c_r")).Transform
   )
   &RigUnit_TransformConstraintPerItem(
    BaseItem=(Type=Bone)
    BaseTransform=(Rotation=(X=0,Y=0,Z=0,W=1),Translation=(X=0,Y=0,Z=0),Scale3D=(X=1,Y=1,Z=1))
    BaseTransformSpace=GlobalSpace
    ExecutePin=()
    Item=(Type=Bone,Name="pec_volume_aim_base_l")
    Targets="((Transform=(Rotation=(X=0,Y=0,Z=0,W=1),Translation=(X=0,Y=0,Z=0),Scale3D=(X=1,Y=1,Z=1)),Weight=0.25,bMaintainOffset=True,Filter=(TranslationFilter=(bX=True,bY=True,bZ=True),RotationFilter=(bX=True,bY=True,bZ=True),ScaleFilter=(bX=True,bY=True,bZ=True))),( ...(truncated)"
    bUseInitialTransforms=True
    Targets.1.Transform=&RigUnit_GetTransform(Item=(Type=Bone,Name="spine_04")).Transform
    Targets.2.Transform=&RigUnit_GetTransform(Item=(Type=Bone,Name="spine_05")).Transform
    Targets.0.Transform=&RigUnit_GetTransform(Item=(Type=Bone,Name="spine_03")).Transform
   )
   &RigUnit_TransformConstraintPerItem(
    BaseItem=(Type=Bone)
    BaseTransform=(Rotation=(X=0,Y=0,Z=0,W=1),Translation=(X=0,Y=0,Z=0),Scale3D=(X=1,Y=1,Z=1))
    BaseTransformSpace=GlobalSpace
    ExecutePin=()
    Item=(Type=Bone,Name="pec_volume_aim_base_r")
    Targets="((Transform=(Rotation=(X=0,Y=0,Z=0,W=1),Translation=(X=0,Y=0,Z=0),Scale3D=(X=1,Y=1,Z=1)),Weight=0.25,bMaintainOffset=True,Filter=(TranslationFilter=(bX=True,bY=True,bZ=True),RotationFilter=(bX=True,bY=True,bZ=True),ScaleFilter=(bX=True,bY=True,bZ=True))),( ...(truncated)"
    bUseInitialTransforms=True
    Targets.2.Transform=&RigUnit_GetTransform(Item=(Type=Bone,Name="spine_05")).Transform
    Targets.1.Transform=&RigUnit_GetTransform(Item=(Type=Bone,Name="spine_04")).Transform
    Targets.0.Transform=&RigUnit_GetTransform(Item=(Type=Bone,Name="spine_03")).Transform
   )
   &RigUnit_TransformConstraintPerItem(
    BaseItem=(Type=Bone)
    BaseTransform=(Rotation=(X=0,Y=0,Z=0,W=1),Translation=(X=0,Y=0,Z=0),Scale3D=(X=1,Y=1,Z=1))
    BaseTransformSpace=GlobalSpace
    ExecutePin=()
    Item=(Type=Bone,Name="bicep_volume_base_l")
    Targets="((Transform=(Rotation=(X=0,Y=0,Z=0,W=1),Translation=(X=0,Y=0,Z=0),Scale3D=(X=1,Y=1,Z=1)),Weight=0.5,bMaintainOffset=True,Filter=(TranslationFilter=(bX=True,bY=True,bZ=True),RotationFilter=(bX=True,bY=True,bZ=True),ScaleFilter=(bX=True,bY=True,bZ=True))),(T ...(truncated)"
    bUseInitialTransforms=True
    Targets.2.Transform=&RigUnit_GetTransform(Item=(Type=Bone,Name="upperarm_twist_01_l")).Transform
    Targets.1.Transform=&RigUnit_GetTransform(Item=(Type=Bone,Name="upperarm_bendy_b_l")).Transform
    Targets.0.Transform=&RigUnit_GetTransform(Item=(Type=Bone,Name="upperarm_bendy_c_l")).Transform
   )
   &RigUnit_TransformConstraintPerItem(
    BaseItem=(Type=Bone)
    BaseTransform=(Rotation=(X=0,Y=0,Z=0,W=1),Translation=(X=0,Y=0,Z=0),Scale3D=(X=1,Y=1,Z=1))
    BaseTransformSpace=GlobalSpace
    ExecutePin=()
    Item=(Type=Bone,Name="bicep_volume_base_r")
    Targets="((Transform=(Rotation=(X=0,Y=0,Z=0,W=1),Translation=(X=0,Y=0,Z=0),Scale3D=(X=1,Y=1,Z=1)),Weight=0.5,bMaintainOffset=True,Filter=(TranslationFilter=(bX=True,bY=True,bZ=True),RotationFilter=(bX=True,bY=True,bZ=True),ScaleFilter=(bX=True,bY=True,bZ=True))),(T ...(truncated)"
    bUseInitialTransforms=True
    Targets.2.Transform=&RigUnit_GetTransform(Item=(Type=Bone,Name="upperarm_twist_01_r")).Transform
    Targets.1.Transform=&RigUnit_GetTransform(Item=(Type=Bone,Name="upperarm_bendy_b_r")).Transform
    Targets.0.Transform=&RigUnit_GetTransform(Item=(Type=Bone,Name="upperarm_bendy_c_r")).Transform
   )
  ]}
]
hierarchy [
 root {loc=<0 0 0>, rot=<0 0 0 1>}
 pelvis {parent=root, loc=<0 -0.7071 93.1191>, rot=<0.0127 0.707 -0.0127 -0.707>}
 spine_01 {parent=pelvis, loc=<6.7494 -0.6388 0>, rot=<0 0 0.034 -0.9994>}
 spine_02 {parent=spine_01, loc=<7.2518 -0.035 0>, rot=<0 0 0.0238 -0.9997>}
 spine_03 {parent=spine_02, loc=<6.0104 0 0>, rot=<0 0 0.0149 -0.9999>}
 spine_04 {parent=spine_03, loc=<8.1024 0 0>, rot=<0 0 -0.0132 -0.9999>}
 spine_05 {parent=spine_04, loc=<10.0146 0 0>, rot=<0 0 -0.0015 -1>}
 clavicle_l {parent=spine_05, loc=<4.3575 -0.1466 -2.3223>, rot=<0.0465 -0.7289 -0.009 -0.6829>}
 upperarm_l {parent=clavicle_l, loc=<15.5111 0 0>, rot=<0.0226 -0.3234 -0.0088 -0.9459>}
 lowerarm_l {parent=upperarm_l, loc=<26.6514 0 0>, rot=<0.0391 -0.0099 -0.2453 -0.9686>}
 hand_l {parent=lowerarm_l, loc=<22.3491 0 0>, rot=<-0.7081 -0.0507 0.0975 -0.6975>}
 index_metacarpal_l {parent=hand_l, loc=<1.6535 0.0183 -2.96>, rot=<0.0158 -0.0124 -0.0173 -0.9996>}
 index_01_l {parent=index_metacarpal_l, loc=<8.3008 0.0357 -0.0524>, rot=<-0.0044 0.0485 0.0924 -0.9945>}
 index_02_l {parent=index_01_l, loc=<3.3587 -0.3004 0.0452>, rot=<-0.002 0.0027 0.3225 -0.9466>}
 index_03_l {parent=index_02_l, loc=<2.5578 -0.123 -0.0252>, rot=<-0.002 0.0067 0.2558 -0.9667>}
 middle_metacarpal_l {parent=hand_l, loc=<2.1856 -0.1725 -0.5984>, rot=<0.006 0.0168 -0.0329 -0.9993>}
 middle_01_l {parent=middle_metacarpal_l, loc=<7.3275 -0.2447 0.0129>, rot=<-0.0208 0.0042 0.0858 -0.9961>}
 middle_02_l {parent=middle_01_l, loc=<4.0123 0.1401 0.1175>, rot=<-0.0218 0.0258 0.3512 -0.9357>}
 middle_03_l {parent=middle_02_l, loc=<2.6077 0.0534 0.1373>, rot=<0.0082 0.0188 0.1708 -0.9851>}
 pinky_metacarpal_l {parent=hand_l, loc=<1.7616 -0.9753 3.3814>, rot=<-0.0397 0.0814 -0.069 -0.9935>}
 pinky_01_l {parent=pinky_metacarpal_l, loc=<7.1709 -0.4831 -0.1094>, rot=<-0.0654 -0.0035 0.2242 -0.9723>}
 pinky_02_l {parent=pinky_01_l, loc=<2.8274 0.0893 -0.1043>, rot=<0.0037 0.0011 0.2854 -0.9584>}
 pinky_03_l {parent=pinky_02_l, loc=<2.2046 0.0307 -0.0078>, rot=<-0.0002 0.0098 0.1445 -0.9895>}
 ring_metacarpal_l {parent=hand_l, loc=<2.2074 -0.3625 1.3638>, rot=<-0.0284 0.0722 -0.0458 -0.9959>}
 ring_01_l {parent=ring_metacarpal_l, loc=<7.0295 -0.6197 -0.1446>, rot=<-0.0489 -0.0252 0.1892 -0.9804>}
 ring_02_l {parent=ring_01_l, loc=<3.6174 0.4647 -0.1027>, rot=<0 0 0.2806 -0.9598>}
 ring_03_l {parent=ring_02_l, loc=<2.4152 0.114 0.0199>, rot=<0 0 0.1189 -0.9929>}
 thumb_01_l {parent=hand_l, loc=<1.3118 -0.3191 -2.459>, rot=<0.6546 -0.4042 0.0664 -0.6354>}
 thumb_02_l {parent=thumb_01_l, loc=<5.9336 -0.0448 0.0401>, rot=<-0.0037 0.0213 0.298 -0.9543>}
 thumb_03_l {parent=thumb_02_l, loc=<2.2223 0.3682 -0.0907>, rot=<0.0118 -0.013 0.0066 -0.9998>}
 lowerarm_twist_01_l {parent=lowerarm_l, loc=<17.8549 0 0>, rot=<0 0 0 -1>}
 lowerarm_bendy_c_base_l {parent=lowerarm_twist_01_l, loc=<0 0 0>, rot=<0 0 0 -1>}
 lowerarm_bendy_c_l {parent=lowerarm_bendy_c_base_l, loc=<0 0 0>, rot=<0 0 0 -1>}
 lowerarm_twist_02_l {parent=lowerarm_l, loc=<9.6199 0 0>, rot=<0 0 0 -1>}
 lowerarm_bendy_b_base_l {parent=lowerarm_twist_02_l, loc=<0 0 0>, rot=<0 0 0 -1>}
 lowerarm_bendy_b_l {parent=lowerarm_bendy_b_base_l, loc=<0 0 0>, rot=<0 0 0 -1>}
 elbow_out_volume_base_l {parent=lowerarm_l, loc=<0 -4 0>, rot=<0 0 0 -1>}
 elbow_out_volume_l {parent=elbow_out_volume_base_l, loc=<0 0 0>, rot=<0 0 0 -1>}
 elbow_in_volume_base_l {parent=lowerarm_l, loc=<0.2112 5.259 0.8274>, rot=<0 0 0 -1>}
 elbow_in_volume_l {parent=elbow_in_volume_base_l, loc=<0 0 0>, rot=<0 0 0 -1>}
 dyn_elbow_ball_base_l {parent=lowerarm_l, loc=<0.2146 -6.0811 -1.3316>, rot=<0.0457 0.3586 0.2347 -0.9023>}
 dyn_elbow_ball_l {parent=dyn_elbow_ball_base_l, loc=<0 0 0>, rot=<0 0 0 1>}
 lowerarm_bendy_a_base_l {parent=lowerarm_l, loc=<0 0 0>, rot=<0 0 0 -1>}
 lowerarm_bendy_a_l {parent=lowerarm_bendy_a_base_l, loc=<0 0 0>, rot=<0 0 0 -1>}
 upperarm_twist_01_l {parent=upperarm_l, loc=<0.0041 0 0>, rot=<0 0 0 -1>}
 upperarm_fr_volume_base_l {parent=upperarm_twist_01_l, loc=<-0.3691 7.0506 -2.765>, rot=<0 0 0 -1>}
 upperarm_fr_volume_l {parent=upperarm_fr_volume_base_l, loc=<0 0 0>, rot=<0 0 0 -1>}
 upperarm_up_volume_base_l {parent=upperarm_twist_01_l, loc=<-3.2216 -1.1391 3.3311>, rot=<0 0 0 -1>}
 upperarm_up_volume_l {parent=upperarm_up_volume_base_l, loc=<0 0 0>, rot=<0 0 0 -1>}
 upperarm_bk_volume_base_l {parent=upperarm_twist_01_l, loc=<1.0813 -7.2193 -1.9453>, rot=<0 0 0 -1>}
 upperarm_bk_volume_l {parent=upperarm_bk_volume_base_l, loc=<0 0 0>, rot=<0 0 0 -1>}
 upperarm_twist_02_l {parent=upperarm_l, loc=<13.7941 0 0>, rot=<0 0 0 -1>}
 upperarm_bendy_b_base_l {parent=upperarm_twist_02_l, loc=<0 0 0>, rot=<0 0 0 -1>}
 upperarm_bendy_b_l {parent=upperarm_bendy_b_base_l, loc=<0 0 0>, rot=<0 0 0 -1>}
 bicep_volume_base_l {parent=upperarm_twist_02_l, loc=<3.3462 0.6099 -1.3508>, rot=<0 0 0 -1>}
 bicep_volume_l {parent=bicep_volume_base_l, loc=<0 0 0>, rot=<0 0 0 -1>}
 upperarm_bendy_c_base_l {parent=upperarm_l, loc=<26.6514 0 0>, rot=<0 0 0 -1>}
 upperarm_bendy_c_l {parent=upperarm_bendy_c_base_l, loc=<0 0 0>, rot=<0 0 0 -1>}
 pose_arm_up_l {parent=clavicle_l, loc=<33.0497 0.0525 20.0671>, rot=<0.0008 0.0336 -0.0252 -0.9991>}
 pose_arm_fwd_l {parent=clavicle_l, loc=<10.909 26.2275 -1.1098>, rot=<0.0008 0.0336 -0.0252 -0.9991>}
 pose_arm_fwd_dn_l {parent=clavicle_l, loc=<12.5366 20.4271 -16.8576>, rot=<0.0008 0.0336 -0.0252 -0.9991>}
 pose_arm_bk_l {parent=clavicle_l, loc=<27.8202 -23.6053 1.2519>, rot=<0.0008 0.0336 -0.0252 -0.9991>}
 clavicle_r {parent=spine_05, loc=<4.3575 -0.1466 2.3223>, rot=<0.7289 0.0465 -0.6829 0.009>}
 upperarm_r {parent=clavicle_r, loc=<-15.5111 0 0>, rot=<0.0226 -0.3234 -0.0088 -0.9459>}
 lowerarm_r {parent=upperarm_r, loc=<-26.6515 0 0>, rot=<0.0391 -0.0099 -0.2453 -0.9686>}
 hand_r {parent=lowerarm_r, loc=<-22.3492 0 0>, rot=<-0.7081 -0.0507 0.0975 -0.6975>}
 index_metacarpal_r {parent=hand_r, loc=<-1.6534 -0.0181 2.9601>, rot=<0.0158 -0.0124 -0.0173 -0.9996>}
 index_01_r {parent=index_metacarpal_r, loc=<-8.3008 -0.0357 0.0524>, rot=<-0.0044 0.0485 0.0924 -0.9945>}
 index_02_r {parent=index_01_r, loc=<-3.3589 0.3003 -0.0453>, rot=<-0.002 0.0027 0.3225 -0.9466>}
 index_03_r {parent=index_02_r, loc=<-2.5578 0.1231 0.0251>, rot=<-0.002 0.0067 0.2558 -0.9667>}
 middle_metacarpal_r {parent=hand_r, loc=<-2.1856 0.1726 0.5984>, rot=<0.006 0.0168 -0.0329 -0.9993>}
 middle_01_r {parent=middle_metacarpal_r, loc=<-7.3273 0.2447 -0.0129>, rot=<-0.0208 0.0042 0.0858 -0.9961>}
 middle_02_r {parent=middle_01_r, loc=<-4.0142 -0.1411 -0.07>, rot=<-0.0217 0.0259 0.3512 -0.9357>}
 middle_03_r {parent=middle_02_r, loc=<-2.6074 -0.0534 -0.1312>, rot=<0.0083 0.0188 0.1708 -0.9851>}
 pinky_metacarpal_r {parent=hand_r, loc=<-1.7617 0.9753 -3.3814>, rot=<-0.0397 0.0814 -0.069 -0.9935>}
 pinky_01_r {parent=pinky_metacarpal_r, loc=<-7.1707 0.4832 0.1094>, rot=<-0.0654 -0.0035 0.2242 -0.9723>}
 pinky_02_r {parent=pinky_01_r, loc=<-2.8274 -0.0892 0.1043>, rot=<0.0037 0.0011 0.2854 -0.9584>}
 pinky_03_r {parent=pinky_02_r, loc=<-2.2045 -0.031 0.0078>, rot=<-0.0002 0.0098 0.1445 -0.9895>}
 ring_metacarpal_r {parent=hand_r, loc=<-2.2073 0.3627 -1.3638>, rot=<-0.0284 0.0722 -0.0458 -0.9959>}
 ring_01_r {parent=ring_metacarpal_r, loc=<-7.0297 0.6197 0.1446>, rot=<-0.0489 -0.0252 0.1892 -0.9804>}
 ring_02_r {parent=ring_01_r, loc=<-3.6171 -0.4648 0.1027>, rot=<0 0 0.2806 -0.9598>}
 ring_03_r {parent=ring_02_r, loc=<-2.4153 -0.1141 -0.0199>, rot=<0 0 0.1189 -0.9929>}
 thumb_01_r {parent=hand_r, loc=<-1.3119 0.3192 2.459>, rot=<0.6546 -0.4042 0.0664 -0.6354>}
 thumb_02_r {parent=thumb_01_r, loc=<-5.9336 0.0447 -0.0401>, rot=<-0.0037 0.0213 0.298 -0.9543>}
 thumb_03_r {parent=thumb_02_r, loc=<-2.2223 -0.3682 0.0907>, rot=<0.0118 -0.013 0.0066 -0.9998>}
 lowerarm_twist_01_r {parent=lowerarm_r, loc=<-17.855 0 0>, rot=<0 0 0 1>}
 lowerarm_bendy_c_base_r {parent=lowerarm_twist_01_r, loc=<0 0 0>, rot=<0 0 0 1>}
 lowerarm_bendy_c_r {parent=lowerarm_bendy_c_base_r, loc=<0 0 0>, rot=<0 0 0 1>}
 lowerarm_twist_02_r {parent=lowerarm_r, loc=<-9.6199 0 0>, rot=<0 0 0 1>}
 lowerarm_bendy_b_base_r {parent=lowerarm_twist_02_r, loc=<0 0 0>, rot=<0 0 0 1>}
 lowerarm_bendy_b_r {parent=lowerarm_bendy_b_base_r, loc=<0 0 0>, rot=<0 0 0 1>}
 elbow_out_volume_base_r {parent=lowerarm_r, loc=<0 4 0>, rot=<0 0 0 1>}
 elbow_out_volume_r {parent=elbow_out_volume_base_r, loc=<0 0 0>, rot=<0 0 0 1>}
 elbow_in_volume_base_r {parent=lowerarm_r, loc=<-0.2111 -5.2591 -0.8275>, rot=<0 0 0 1>}
 elbow_in_volume_r {parent=elbow_in_volume_base_r, loc=<-0.0001 0 0.0002>, rot=<0 0 0 1>}
 dyn_elbow_ball_base_r {parent=lowerarm_r, loc=<-0.2147 6.0812 1.3319>, rot=<0.0457 0.3586 0.2347 -0.9023>}
 dyn_elbow_ball_r {parent=dyn_elbow_ball_base_r, loc=<0 0 0>, rot=<0 0 0 1>}
 lowerarm_bendy_a_base_r {parent=lowerarm_r, loc=<0 0 0>, rot=<0 0 0 1>}
 lowerarm_bendy_a_r {parent=lowerarm_bendy_a_base_r, loc=<0 0 0>, rot=<0 0 0 1>}
 upperarm_twist_01_r {parent=upperarm_r, loc=<-0.0041 0 0>, rot=<0 0 0 1>}
 upperarm_fr_volume_base_r {parent=upperarm_twist_01_r, loc=<0.3689 -7.0506 2.7652>, rot=<0 0 0 1>}
 upperarm_fr_volume_r {parent=upperarm_fr_volume_base_r, loc=<0 0 0>, rot=<0 0 0 1>}
 upperarm_up_volume_base_r {parent=upperarm_twist_01_r, loc=<3.2213 1.1391 -3.3308>, rot=<0 0 0 1>}
 upperarm_up_volume_r {parent=upperarm_up_volume_base_r, loc=<0 0 0>, rot=<0 0 0 -1>}
 upperarm_bk_volume_base_r {parent=upperarm_twist_01_r, loc=<-1.0815 7.2192 1.9456>, rot=<0 0 0 1>}
 upperarm_bk_volume_r {parent=upperarm_bk_volume_base_r, loc=<0 0 0>, rot=<0 0 0 1>}
 upperarm_twist_02_r {parent=upperarm_r, loc=<-13.7942 0 0>, rot=<0 0 0 1>}
 upperarm_bendy_b_base_r {parent=upperarm_twist_02_r, loc=<0 0 0>, rot=<0 0 0 1>}
 upperarm_bendy_b_r {parent=upperarm_bendy_b_base_r, loc=<0 0 0>, rot=<0 0 0 1>}
 bicep_volume_base_r {parent=upperarm_twist_02_r, loc=<-3.3462 -0.4709 1.0435>, rot=<0 0 0 1>}
 bicep_volume_r {parent=bicep_volume_base_r, loc=<0 0 0>, rot=<0 0 0 1>}
 upperarm_bendy_c_base_r {parent=upperarm_r, loc=<-26.6515 0 0>, rot=<0 0 0 1>}
 upperarm_bendy_c_r {parent=upperarm_bendy_c_base_r, loc=<0 0 0>, rot=<0 0 0 1>}
 pose_arm_up_r {parent=clavicle_r, loc=<-33.0497 -0.0525 -20.0672>, rot=<0.9991 0.0252 0.0336 0.0008>}
 pose_arm_bk_r {parent=clavicle_r, loc=<-27.8203 23.6054 -1.2519>, rot=<0.9991 0.0252 0.0336 0.0008>}
 pose_arm_fwd_r {parent=clavicle_r, loc=<-10.909 -26.2277 1.1098>, rot=<0.9991 0.0252 0.0336 0.0008>}
 pose_arm_fwd_dn_r {parent=clavicle_r, loc=<-12.5366 -20.4272 16.8577>, rot=<0.9991 0.0252 0.0336 0.0008>}
 neck_01 {parent=spine_05, loc=<11.3138 0.1529 0>, rot=<0 0 -0.2003 -0.9797>}
 neck_02 {parent=neck_01, loc=<4.5627 0 0>, rot=<0 0 0.0504 -0.9987>}
 head {parent=neck_02, loc=<5.3914 0 0>, rot=<0 0 0.115 -0.9934>}
 pec_volume_aim_base_r {parent=spine_04, loc=<7.8727 12.6388 2.4123>, rot=<-0.5669 -0.2754 0.7444 0.2208>}
 pec_volume_aim_r {parent=pec_volume_aim_base_r, loc=<0 0 0>, rot=<0 0 0 -1>}
 pec_volume_base_r {parent=pec_volume_aim_r, loc=<-10.766 -2.6868 0.7511>, rot=<-0.1959 -0.0574 -0.1126 -0.9724>}
 pec_volume_r {parent=pec_volume_base_r, loc=<0 0 0>, rot=<0 0 0 1>}
 pec_volume_aim_base_l {parent=spine_04, loc=<7.8732 12.6388 -2.4123>, rot=<0.2754 -0.5669 0.2208 -0.7444>}
 pec_volume_aim_l {parent=pec_volume_aim_base_l, loc=<0 0 0>, rot=<0 0 0 1>}
 pec_volume_base_l {parent=pec_volume_aim_l, loc=<10.766 2.6868 -0.7511>, rot=<-0.1959 -0.0574 -0.1126 -0.9724>}
 pec_volume_l {parent=pec_volume_base_l, loc=<0 0 0>, rot=<0 0 0 1>}
 thigh_l {parent=pelvis, loc=<-2.5693 -0.0109 -7.7501>, rot=<0.0012 0.0572 0.0155 -0.9982>}
 calf_l {parent=thigh_l, loc=<-39.0484 0 0>, rot=<0.0006 -0.0161 -0.0391 -0.9991>}
 calf_twist_01_l {parent=calf_l, loc=<-35.8131 -0.0967 0.0397>, rot=<0.0237 0.0095 -0.0035 -0.9997>}
 lowerleg_bendy_c_base_l {parent=calf_twist_01_l, loc=<0 0 0>, rot=<0 0 0 -1>}
 lowerleg_bendy_c_l {parent=lowerleg_bendy_c_base_l, loc=<0 0 0>, rot=<0 0 0 -1>}
 calf_twist_02_l {parent=calf_l, loc=<-18.8064 -0.0967 0.0397>, rot=<0.0237 0.0095 -0.0035 -0.9997>}
 lowerleg_bendy_b_base_l {parent=calf_twist_02_l, loc=<0 0 0>, rot=<0 0 0 -1>}
 lowerleg_bendy_b_l {parent=lowerleg_bendy_b_base_l, loc=<0 0 0>, rot=<0 0 0 -1>}
 foot_l {parent=calf_l, loc=<-42.561 0.1077 0.4292>, rot=<0.0418 -0.0455 0.0343 -0.9975>}
 ball_l {parent=foot_l, loc=<-6.1599 13.8264 0.6806>, rot=<0.0001 -0.0001 -0.7063 -0.7079>}
 knee_out_volume_base_l {parent=calf_l, loc=<0 4 0>, rot=<0 0 0 1>}
 knee_out_volume_l {parent=knee_out_volume_base_l, loc=<0 0 0>, rot=<0 0 0 1>}
 knee_in_volume_base_l {parent=calf_l, loc=<1.298 -5.2957 -0.0496>, rot=<0 0 0 1>}
 knee_in_volume_l {parent=knee_in_volume_base_l, loc=<0 0 0>, rot=<0 0 0 1>}
 dyn_knee_ball_base_l {parent=calf_l, loc=<0.0192 4.6633 -0.2461>, rot=<-0.0315 -0.7349 0.027 -0.6769>}
 dyn_knee_ball_l {parent=dyn_knee_ball_base_l, loc=<0 0 0>, rot=<0 0 0 1>}
 lowerleg_bendy_a_base_l {parent=calf_l, loc=<0 0 0>, rot=<0 0 0 1>}
 lowerleg_bendy_a_l {parent=lowerleg_bendy_a_base_l, loc=<0 0 0>, rot=<0 0 0 1>}
 thigh_twist_01_l {parent=thigh_l, loc=<-21.6303 0 0>, rot=<0.0013 0 0 -1>}
 upperleg_bendy_b_base_l {parent=thigh_twist_01_l, loc=<0 0 0>, rot=<0 0 0 1>}
 upperleg_bendy_b_l {parent=upperleg_bendy_b_base_l, loc=<0 0 0>, rot=<0 0 0 1>}
 glute_volume_base_l {parent=thigh_l, loc=<-6 -9 0>, rot=<0 0 -0.4617 -0.887>}
 glute_volume_l {parent=glute_volume_base_l, loc=<0 0 0>, rot=<0 0 0 -1>}
 crouch_volume_base_l {parent=thigh_l, loc=<-2.1908 8.1142 -1.7442>, rot=<0 0 0 -1>}
 crouch_volume_l {parent=crouch_volume_base_l, loc=<0 0 0>, rot=<0 0 0 1>}
 dyn_gun_base_l {parent=thigh_l, loc=<-9.1369 0.8479 -6.4403>, rot=<-0.0036 -0.9988 -0.0367 0.0314>}
 dyn_gun_l {parent=dyn_gun_base_l, loc=<0 0 0>, rot=<0 0 0 -1>}
 upperleg_bendy_c_base_l {parent=thigh_l, loc=<-39.0484 0 0>, rot=<0 0 0 1>}
 upperleg_bendy_c_l {parent=upperleg_bendy_c_base_l, loc=<0 0 0>, rot=<0 0 0 1>}
 thigh_r {parent=pelvis, loc=<-2.5691 -0.0109 7.7501>, rot=<0.0572 -0.0012 0.9982 0.0155>}
 calf_r {parent=thigh_r, loc=<39.0487 0 0>, rot=<0.0006 -0.0161 -0.0391 -0.9991>}
 calf_twist_01_r {parent=calf_r, loc=<35.8132 0.0967 -0.0397>, rot=<0.0237 0.0095 -0.0035 -0.9997>}
 lowerleg_bendy_c_base_r {parent=calf_twist_01_r, loc=<0 0 0>, rot=<0 0 0 1>}
 lowerleg_bendy_c_r {parent=lowerleg_bendy_c_base_r, loc=<0 0 0>, rot=<0 0 0 1>}
 calf_twist_02_r {parent=calf_r, loc=<18.8064 0.0967 -0.0397>, rot=<0.0237 0.0095 -0.0035 -0.9997>}
 lowerleg_bendy_b_base_r {parent=calf_twist_02_r, loc=<0 0 0>, rot=<0 0 0 1>}
 lowerleg_bendy_b_r {parent=lowerleg_bendy_b_base_r, loc=<0 0 0>, rot=<0 0 0 1>}
 foot_r {parent=calf_r, loc=<42.5612 -0.1077 -0.4292>, rot=<0.0418 -0.0455 0.0343 -0.9975>}
 ball_r {parent=foot_r, loc=<6.1598 -13.8264 -0.6806>, rot=<0.0001 -0.0001 -0.7063 -0.7079>}
 knee_out_volume_base_r {parent=calf_r, loc=<0 -4 0>, rot=<0 0 0 1>}
 knee_out_volume_r {parent=knee_out_volume_base_r, loc=<0 0 0>, rot=<0 0 0 1>}
 knee_in_volume_base_r {parent=calf_r, loc=<-1.2981 5.2958 0.0496>, rot=<0 0 0 1>}
 knee_in_volume_r {parent=knee_in_volume_base_r, loc=<0 0 0>, rot=<0 0 0 1>}
 dyn_knee_ball_base_r {parent=calf_r, loc=<-0.0192 -4.6633 0.246>, rot=<-0.0315 -0.7349 0.027 -0.6769>}
 dyn_knee_ball_r {parent=dyn_knee_ball_base_r, loc=<0 0 0>, rot=<0 0 0 1>}
 lowerleg_bendy_a_base_r {parent=calf_r, loc=<0 0 0>, rot=<0 0 0 1>}
 lowerleg_bendy_a_r {parent=lowerleg_bendy_a_base_r, loc=<0 0 0>, rot=<0 0 0 1>}
 thigh_twist_01_r {parent=thigh_r, loc=<21.6303 0 0>, rot=<0.0013 0 0 -1>}
 upperleg_bendy_b_base_r {parent=thigh_twist_01_r, loc=<0 0 0>, rot=<0 0 0 1>}
 upperleg_bendy_b_r {parent=upperleg_bendy_b_base_r, loc=<0 0 0>, rot=<0 0 0 1>}
 glute_volume_base_r {parent=thigh_r, loc=<6 9 0>, rot=<0 0 -0.4617 -0.887>}
 glute_volume_r {parent=glute_volume_base_r, loc=<0 0 0>, rot=<0 0 0 1>}
 crouch_volume_base_r {parent=thigh_r, loc=<2.1909 -8.1142 1.7442>, rot=<0 0 0 -1>}
 crouch_volume_r {parent=crouch_volume_base_r, loc=<0 0 0>, rot=<0 0 0 1>}
 dyn_gun_base_r {parent=thigh_r, loc=<9.1371 -0.8479 6.4403>, rot=<-0.0036 -0.9988 -0.0367 0.0314>}
 dyn_gun_r {parent=dyn_gun_base_r, loc=<0 0 0>, rot=<0 0 0 1>}
 upperleg_bendy_c_base_r {parent=thigh_r, loc=<39.0487 0 0>, rot=<0 0 0 1>}
 upperleg_bendy_c_r {parent=upperleg_bendy_c_base_r, loc=<0 0 0>, rot=<0 0 0 1>}
 dyn_button_base_l {parent=pelvis, loc=<0.1519 10.3262 -7.7447>, rot=<0.1298 -0.6976 0.1415 -0.6902>}
 dyn_button_l {parent=dyn_button_base_l, loc=<0 0 0>, rot=<0 0 0 -1>}
 dyn_button_base_r {parent=pelvis, loc=<0.1519 10.3262 7.7447>, rot=<0.6976 0.1298 -0.6902 -0.1415>}
 dyn_button_r {parent=dyn_button_base_r, loc=<0 0 0>, rot=<0 0 0 -1>}
]
variables [
 elbow_rot_l {type=double, default=0}
 elbow_rot_r {type=double, default=0}
 knee_rot_l {type=double, default=0}
 knee_rot_r {type=double, default=0}
 psd_arm_bk_alpha_l {type=double, default=0}
 psd_arm_bk_alpha_r {type=double, default=0}
 psd_arm_fwd_alpha_l {type=double, default=0}
 psd_arm_fwd_alpha_r {type=double, default=0}
 psd_arm_fwd_dn_alpha_l {type=double, default=0}
 psd_arm_fwd_dn_alpha_r {type=double, default=0}
 psd_arm_up_alpha_l {type=double, default=0}
 psd_arm_up_alpha_r {type=double, default=0}
 thigh_rot_l {type=double, default=0}
 thigh_rot_r {type=double, default=0}
]
