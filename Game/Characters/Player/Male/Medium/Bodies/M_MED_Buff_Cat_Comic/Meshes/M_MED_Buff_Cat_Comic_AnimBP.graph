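# Body AnimBP for the Buff Cat Comic player character, adding joint-volume correctives and secondary motion to a pose copied from the parent mesh.
# EventGraph derives joint yaw, pose-space arm alphas and a crouch alpha.
# An emote bend layer feeds volume_ControlRig, then constraints, driven bone offsets and gun AnimDynamics on the way to output.
# A node-based volume chain, cached as volumePose, does not reach the output.
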AnimBlueprint /Game/Characters/Player/Male/Medium/Bodies/M_MED_Buff_Cat_Comic/Meshes/M_MED_Buff_Cat_Comic_AnimBP
target_skeleton=&/M_MED_Buff_Cat_Comic_Skeleton
graphs [
 AnimGraph {nodes [
   n1 = "Constraint - knee_in_volume_base_l (PP)"!(ComponentPose, Alpha=1, ConstraintWeights_0=0.5, ConstraintWeights_1:float=0.5) -> (Pose) {
    alphaBoolBlend {blendOption=Linear}
    boneToModify {boneName=knee_in_volume_base_l}
    constraintSetup [{targetBone {boneName=thigh_l}}, {targetBone {boneName=calf_l}}]
    constraintWeights [1,1]
   }
   "Constraint - knee_in_volume_base_r (PP)"!(ComponentPose=@n1, Alpha=1, ConstraintWeights_0=0.5, ConstraintWeights_1:float=0.5) -> (Pose) {
    alphaBoolBlend {blendOption=Linear}
    boneToModify {boneName=knee_in_volume_base_r}
    constraintSetup [{targetBone {boneName=thigh_r}}, {targetBone {boneName=calf_r}}]
    constraintWeights [1,1]
   }
   n2 = "Constraint - elbow_out_volume_base_l (PP)"!(ComponentPose, Alpha=1, ConstraintWeights_0=0.65, ConstraintWeights_1:float=0.35) -> (Pose) {
    alphaBoolBlend {blendOption=Linear}
    boneToModify {boneName=elbow_out_volume_base_l}
    constraintSetup [{targetBone {boneName=upperarm_l}}, {targetBone {boneName=lowerarm_l}}]
    constraintWeights [1,1]
   }
   "Constraint - elbow_out_volume_base_r (PP)"!(ComponentPose=@n2, Alpha=1, ConstraintWeights_0=0.65, ConstraintWeights_1:float=0.35) -> (Pose) {
    alphaBoolBlend {blendOption=Linear}
    boneToModify {boneName=elbow_out_volume_base_r}
    constraintSetup [{targetBone {boneName=upperarm_r}}, {targetBone {boneName=lowerarm_r}}]
    constraintWeights [1,1]
   }
   n3 = "Constraint - knee_out_volume_base_l (PP)"!(ComponentPose, Alpha=1, ConstraintWeights_0=0.5, ConstraintWeights_1:float=0.5) -> (Pose) {
    alphaBoolBlend {blendOption=Linear}
    boneToModify {boneName=knee_out_volume_base_l}
    constraintSetup [{targetBone {boneName=thigh_l}}, {targetBone {boneName=calf_l}}]
    constraintWeights [1,1]
   }
   "Constraint - knee_out_volume_base_r (PP)"!(ComponentPose=@n3, Alpha=1, ConstraintWeights_0=0.5, ConstraintWeights_1:float=0.5) -> (Pose) {
    alphaBoolBlend {blendOption=Linear}
    boneToModify {boneName=knee_out_volume_base_r}
    constraintSetup [{targetBone {boneName=thigh_r}}, {targetBone {boneName=calf_r}}]
    constraintWeights [1,1]
   }
   n4 = "Constraint - elbow_in_volume_base_l (PP)"!(ComponentPose, Alpha=1, ConstraintWeights_0=0.6, ConstraintWeights_1:float=0.4) -> (Pose) {
    alphaBoolBlend {blendOption=Linear}
    boneToModify {boneName=elbow_in_volume_base_l}
    constraintSetup [{targetBone {boneName=upperarm_l}}, {targetBone {boneName=lowerarm_l}}]
    constraintWeights [1,1]
   }
   "Constraint - elbow_in_volume_base_r (PP)"!(ComponentPose=@n4, Alpha=1, ConstraintWeights_0=0.6, ConstraintWeights_1:float=0.4) -> (Pose) {
    alphaBoolBlend {blendOption=Linear}
    boneToModify {boneName=elbow_in_volume_base_r}
    constraintSetup [{targetBone {boneName=upperarm_r}}, {targetBone {boneName=lowerarm_r}}]
    constraintWeights [1,1]
   }
   n5 = "Constraint - dyn_gun_base_l (PP)"!(
    ComponentPose=Local To Component(LocalPose=Use cached pose 'EmoteBlend_pose'())
    Alpha=1
    ConstraintWeights_0=0.2
    ConstraintWeights_1:float=0.8
   ) -> (Pose) {
    alphaBoolBlend {blendOption=Linear}
    boneToModify {boneName=dyn_gun_base_l}
    constraintSetup [{targetBone {boneName=pelvis}}, {targetBone {boneName=upperleg_bendy_b_l}}]
    constraintWeights [1,1]
   }
   n6 = "Constraint - dyn_gun_base_r (PP)"!(ComponentPose=@n5, Alpha=1, ConstraintWeights_0=0.2, ConstraintWeights_1:float=0.8) -> (Pose) {
    alphaBoolBlend {blendOption=Linear}
    boneToModify {boneName=dyn_gun_base_r}
    constraintSetup [{targetBone {boneName=pelvis}}, {targetBone {boneName=upperleg_bendy_b_r}}]
    constraintWeights [1,1]
   }
   n7 = "Constraint - dyn_knee_ball_base_l (PPP)"!(ComponentPose=@n6, Alpha=1, ConstraintWeights_0=0.64, ConstraintWeights_1:float=0.23, ConstraintWeights_2:float=0.13) -> (Pose) {
    alphaBoolBlend {blendOption=Linear}
    boneToModify {boneName=dyn_knee_ball_base_l}
    constraintSetup [{targetBone {boneName=knee_out_volume_l}}, {targetBone {boneName=upperleg_bendy_c_l}}, {targetBone {boneName=lowerleg_bendy_a_l}}]
    constraintWeights [1,1,1]
   }
   n8 = "Constraint - dyn_knee_ball_base_r (PPP)"!(ComponentPose=@n7, Alpha=1, ConstraintWeights_0=0.13, ConstraintWeights_1:float=0.23, ConstraintWeights_2:float=0.64) -> (Pose) {
    alphaBoolBlend {blendOption=Linear}
    boneToModify {boneName=dyn_knee_ball_base_r}
    constraintSetup [{targetBone {boneName=lowerleg_bendy_a_r}}, {targetBone {boneName=upperleg_bendy_c_r}}, {targetBone {boneName=knee_out_volume_r}}]
    constraintWeights [1,1,1]
   }
   n9 = "Constraint - dyn_elbow_ball_base_l (PPPPP)"!(
    ComponentPose=@n8
    Alpha=1
    ConstraintWeights_0=0.12
    ConstraintWeights_1:float=0.01
    ConstraintWeights_2:float=0.54
    ConstraintWeights_3:float=0.16
    ConstraintWeights_4:float=0.17
   ) -> (Pose) {
    alphaBoolBlend {blendOption=Linear}
    boneToModify {boneName=dyn_elbow_ball_base_l}
    constraintSetup [
     {targetBone {boneName=upperarm_bendy_c_l}}
     {targetBone {boneName=upperarm_bendy_b_l}}
     {targetBone {boneName=elbow_out_volume_l}}
     {targetBone {boneName=lowerarm_bendy_b_l}}
     {targetBone {boneName=lowerarm_bendy_a_l}}
    ]
    constraintWeights [1,1,1,1,1]
   }
   n10 = "Constraint - dyn_elbow_ball_base_r (PPPPP)"!(
    ComponentPose=@n9
    Alpha=1
    ConstraintWeights_0=0.17
    ConstraintWeights_1:float=0.54
    ConstraintWeights_2:float=0.12
    ConstraintWeights_3:float=0.16
    ConstraintWeights_4:float=0.01
   ) -> (Pose) {
    alphaBoolBlend {blendOption=Linear}
    boneToModify {boneName=dyn_elbow_ball_base_r}
    constraintSetup [
     {targetBone {boneName=lowerarm_bendy_a_r}}
     {targetBone {boneName=elbow_out_volume_r}}
     {targetBone {boneName=upperarm_bendy_c_r}}
     {targetBone {boneName=lowerarm_bendy_b_r}}
     {targetBone {boneName=upperarm_bendy_b_r}}
    ]
    constraintWeights [1,1,1,1,1]
   }
   n11 = "Transform (Modify) Bone - Bone: dyn_knee_ball_l"(ComponentPose=@n10, Alpha=Get knee_rot_l() -> (knee_rot_l:double), Translation="0,-1,-1.5", Rotation="0,0,0", Scale="1,1,1") {
    rotationMode=BMM_Additive
    rotationSpace=BCS_BoneSpace
    scaleMode=BMM_Additive
    scaleSpace=BCS_BoneSpace
    translationMode=BMM_Additive
    translationSpace=BCS_BoneSpace
    alphaBoolBlend {blendOption=Linear}
    alphaScaleBiasClamp {bMapRange=true, inRange {max=100}}
    boneToModify {boneName=dyn_knee_ball_l}
    translation {y=-1, z=-1.5}
   }
   n12 = "Transform (Modify) Bone - Bone: dyn_knee_ball_r"(ComponentPose=@n11, Alpha=Get knee_rot_r() -> (knee_rot_r:double), Translation="0,1,1.5", Rotation="0,0,0", Scale="1,1,1") {
    rotationMode=BMM_Additive
    rotationSpace=BCS_BoneSpace
    scaleMode=BMM_Additive
    scaleSpace=BCS_BoneSpace
    translationMode=BMM_Additive
    translationSpace=BCS_BoneSpace
    alphaBoolBlend {blendOption=Linear}
    alphaScaleBiasClamp {bMapRange=true, inRange {max=100}}
    boneToModify {boneName=dyn_knee_ball_r}
    translation {y=1, z=1.5}
   }
   n13 = "Constraint - dyn_button_base_l (PPP)"!(ComponentPose=@n12, Alpha=1, ConstraintWeights_0=0.8, ConstraintWeights_1:float=0.1, ConstraintWeights_2:float=0.1) -> (Pose) {
    alphaBoolBlend {blendOption=Linear}
    boneToModify {boneName=dyn_button_base_l}
    constraintSetup [{targetBone {boneName=pelvis}}, {targetBone {boneName=upperleg_bendy_b_l}}, {targetBone {boneName=crouch_volume_l}}]
    constraintWeights [1,1,1]
   }
   n14 = "Constraint - dyn_button_base_r (PPP)"!(ComponentPose=@n13, Alpha=1, ConstraintWeights_0=0.8, ConstraintWeights_1:float=0.1, ConstraintWeights_2:float=0.1) -> (Pose) {
    alphaBoolBlend {blendOption=Linear}
    boneToModify {boneName=dyn_button_base_r}
    constraintSetup [{targetBone {boneName=pelvis}}, {targetBone {boneName=crouch_volume_r}}, {targetBone {boneName=upperleg_bendy_b_r}}]
    constraintWeights [1,1,1]
   }
   n15 = "Transform (Modify) Bone - Bone: dyn_button_l"(ComponentPose=@n14, Alpha=Get thigh_rot_l(), Translation="-0.5,1,1", Rotation="0.000,0.000,30", Scale="1,1,1") {
    rotationMode=BMM_Additive
    rotationSpace=BCS_BoneSpace
    scaleMode=BMM_Additive
    scaleSpace=BCS_BoneSpace
    translationMode=BMM_Additive
    translationSpace=BCS_BoneSpace
    alphaBoolBlend {blendOption=Linear}
    alphaScaleBiasClamp {bMapRange=true, inRange {max=-20, min=-100}, outRange {max=0, min=1}}
    boneToModify {boneName=dyn_button_l}
    rotation {roll=30}
    translation=<-0.5 1 1>
   }
   n16 = "Transform (Modify) Bone - Bone: dyn_button_l"(ComponentPose=@n15, Alpha=Get thigh_rot_r(), Translation="0.5,-1,-1", Rotation="0.000,0.000,30", Scale="1,1,1") {
    rotationMode=BMM_Additive
    rotationSpace=BCS_BoneSpace
    scaleMode=BMM_Additive
    scaleSpace=BCS_BoneSpace
    translationMode=BMM_Additive
    translationSpace=BCS_BoneSpace
    alphaBoolBlend {blendOption=Linear}
    alphaScaleBiasClamp {bMapRange=true, inRange {max=-20, min=-100}, outRange {max=0, min=1}}
    boneToModify {boneName=dyn_button_l}
    rotation {roll=30}
    translation=<0.5 -1 -1>
   }
   n17 = Get isCrouchingAlpha() -> (isCrouchingAlpha:double)
   n18 = Anim Dynamics - Bone: dyn_gun_l(ComponentPose=@n16, Alpha=@n17, ExternalForce="0,0,0") {
    angularDampingOverride=1
    angularSpringConstant=1500
    bAngularSpring=true
    bGravityOverrideInSimSpace=true
    bOverrideAngularDamping=true
    gravityScale=0.35
    simulationSpace=RootRelative
    alphaBoolBlend {blendOption=Linear}
    alphaScaleBiasClamp {bInterpResult=true, bMapRange=true, interpSpeedDecreasing=5, interpSpeedIncreasing=5, outRange {max=0.2, min=0.6}}
    boundBone {boneName=dyn_gun_l}
    physicsBodyDefinitions [{
      sphereCollisionRadius=0
      boundBone {boneName=dyn_gun_l}
      boxExtents=<10 20 20>
      constraintSetup {angularLimitsMax=<10 10 10>, angularLimitsMin=<-10 -10 -10>, angularTarget=<0 0 0>}
      localJointOffset=<10 0 0>
     }]
   }
   n19 = Anim Dynamics - Bone: dyn_gun_r(ComponentPose=@n18, Alpha=@n17, ExternalForce="0,0,0") {
    angularDampingOverride=1
    angularSpringConstant=1500
    bAngularSpring=true
    bGravityOverrideInSimSpace=true
    bOverrideAngularDamping=true
    gravityScale=0.35
    simulationSpace=RootRelative
    alphaBoolBlend {blendOption=Linear}
    alphaScaleBiasClamp {bInterpResult=true, bMapRange=true, interpSpeedDecreasing=5, interpSpeedIncreasing=5, outRange {max=0.2, min=0.6}}
    boundBone {boneName=dyn_gun_r}
    physicsBodyDefinitions [{
      sphereCollisionRadius=0
      boundBone {boneName=dyn_gun_r}
      boxExtents=<10 20 20>
      constraintSetup {angularLimitsMax=<5 5 10>, angularLimitsMin=<-5 -5 -10>, angularTarget=<0 0 0>}
      localJointOffset=<10 0 0>
     }]
   }
   Output Pose(Result=Component To Local(ComponentPose=@n19)) {layerGroup=None}
   n20 = "Constraint - elbow_out_volume_base_l (PP)"(ComponentPose=Local To Component(), Alpha=1, ConstraintWeights_0=0.6, ConstraintWeights_1=0.4) {
    alphaBoolBlend {blendOption=Linear}
    boneToModify {boneName=elbow_out_volume_base_l}
    constraintSetup [{targetBone {boneName=upperarm_bendy_c_l}}, {targetBone {boneName=lowerarm_bendy_a_l}}]
    constraintWeights [1,1]
   }
   n21 = "Constraint - elbow_out_volume_base_r (PP)"(ComponentPose=@n20, Alpha=1, ConstraintWeights_0=0.4, ConstraintWeights_1=0.6) {
    alphaBoolBlend {blendOption=Linear}
    boneToModify {boneName=elbow_out_volume_base_r}
    constraintSetup [{targetBone {boneName=lowerarm_bendy_a_r}}, {targetBone {boneName=upperarm_bendy_c_r}}]
    constraintWeights [1,1]
   }
   n22 = "Transform (Modify) Bone - Bone: elbow_out_volume_l"(ComponentPose=@n21, Alpha=Get elbow_rot_l() -> (elbow_rot_l:double), Translation="0,-3,0", Rotation="0,0,0", Scale="0.3,1,1") {
    rotationMode=BMM_Additive
    rotationSpace=BCS_BoneSpace
    scaleMode=BMM_Additive
    scaleSpace=BCS_BoneSpace
    translationMode=BMM_Additive
    translationSpace=BCS_BoneSpace
    alphaBoolBlend {blendOption=Linear}
    alphaScaleBiasClamp {bMapRange=true, inRange {max=80}}
    boneToModify {boneName=elbow_out_volume_l}
    scale {x=0.3}
    translation {y=-3}
   }
   n23 = "Transform (Modify) Bone - Bone: elbow_out_volume_r"(ComponentPose=@n22, Alpha=Get elbow_rot_r() -> (elbow_rot_r:double), Translation="0,3,0", Rotation="0,0,0", Scale="0.3,1,1") {
    rotationMode=BMM_Additive
    rotationSpace=BCS_BoneSpace
    scaleMode=BMM_Additive
    scaleSpace=BCS_BoneSpace
    translationMode=BMM_Additive
    translationSpace=BCS_BoneSpace
    alphaBoolBlend {blendOption=Linear}
    alphaScaleBiasClamp {bMapRange=true, inRange {max=80}}
    boneToModify {boneName=elbow_out_volume_r}
    scale {x=0.3}
    translation {y=3}
   }
   n24 = "Constraint - elbow_in_volume_base_l (PP)"(ComponentPose=@n23, Alpha=1, ConstraintWeights_0=0.6, ConstraintWeights_1=0.4) {
    alphaBoolBlend {blendOption=Linear}
    boneToModify {boneName=elbow_in_volume_base_l}
    constraintSetup [{targetBone {boneName=upperarm_bendy_c_l}}, {targetBone {boneName=lowerarm_bendy_a_l}}]
    constraintWeights [1,1]
   }
   n25 = "Constraint - elbow_in_volume_base_r (PP)"(ComponentPose=@n24, Alpha=1, ConstraintWeights_0=0.4, ConstraintWeights_1=0.6) {
    alphaBoolBlend {blendOption=Linear}
    boneToModify {boneName=elbow_in_volume_base_r}
    constraintSetup [{targetBone {boneName=lowerarm_bendy_a_r}}, {targetBone {boneName=upperarm_bendy_c_r}}]
    constraintWeights [1,1]
   }
   n26 = "Transform (Modify) Bone - Bone: elbow_in_volume_l"(ComponentPose=@n25, Alpha=Get elbow_rot_l(), Translation="0,8,0", Rotation="0,0,0", Scale="0.5,1,1") {
    rotationMode=BMM_Additive
    rotationSpace=BCS_BoneSpace
    scaleMode=BMM_Additive
    scaleSpace=BCS_BoneSpace
    translationMode=BMM_Additive
    translationSpace=BCS_BoneSpace
    alphaBoolBlend {blendOption=Linear}
    alphaScaleBiasClamp {bMapRange=true, inRange {max=90}}
    boneToModify {boneName=elbow_in_volume_l}
    scale {x=0.5}
    translation {y=8}
   }
   n27 = "Transform (Modify) Bone - Bone: elbow_in_volume_r"(ComponentPose=@n26, Alpha=Get elbow_rot_r(), Translation="0,-7,0", Rotation="0,0,0", Scale="0.5,1,1") {
    rotationMode=BMM_Additive
    rotationSpace=BCS_BoneSpace
    scaleMode=BMM_Additive
    scaleSpace=BCS_BoneSpace
    translationMode=BMM_Additive
    translationSpace=BCS_BoneSpace
    alphaBoolBlend {blendOption=Linear}
    alphaScaleBiasClamp {bMapRange=true, inRange {max=90}}
    boneToModify {boneName=elbow_in_volume_r}
    scale {x=0.5}
    translation {y=-7}
   }
   n28 = "Constraint - glute_volume_base_l (PP)"!(ComponentPose=@n27, Alpha=1, ConstraintWeights_0=0.6, ConstraintWeights_1:float=0.4) -> (Pose) {
    alphaBoolBlend {blendOption=Linear}
    boneToModify {boneName=glute_volume_base_l}
    constraintSetup [{targetBone {boneName=pelvis}}, {targetBone {boneName=thigh_l}}]
    constraintWeights [1,1]
   }
   n29 = "Constraint - glute_volume_base_r (PP)"!(ComponentPose=@n28, Alpha=1, ConstraintWeights_0=0.6, ConstraintWeights_1:float=0.4) -> (Pose) {
    alphaBoolBlend {blendOption=Linear}
    boneToModify {boneName=glute_volume_base_r}
    constraintSetup [{targetBone {boneName=pelvis}}, {targetBone {boneName=thigh_r}}]
    constraintWeights [1,1]
   }
   n30 = Get thigh_rot_l()
   n31 = "Transform (Modify) Bone - Bone: glute_volume_l"(ComponentPose=@n29, Alpha=@n30, Translation="-10,-5,0", Rotation="0,0,0", Scale="1,1,1") {
    rotationMode=BMM_Additive
    rotationSpace=BCS_BoneSpace
    scaleMode=BMM_Additive
    scaleSpace=BCS_BoneSpace
    translationMode=BMM_Additive
    translationSpace=BCS_BoneSpace
    alphaBoolBlend {blendOption=Linear}
    alphaScaleBiasClamp {bMapRange=true, inRange {max=0, min=-100}, outRange {max=0, min=1}}
    boneToModify {boneName=glute_volume_l}
    translation=<-10 -5>
   }
   n32 = "Transform (Modify) Bone - Bone: glute_volume_l"(ComponentPose=@n31, Alpha=@n30, Translation="0,-4,0", Rotation="0,0,0", Scale="1,1,1") {
    rotationMode=BMM_Additive
    rotationSpace=BCS_BoneSpace
    scaleMode=BMM_Additive
    scaleSpace=BCS_BoneSpace
    translationMode=BMM_Additive
    translationSpace=BCS_BoneSpace
    alphaBoolBlend {blendOption=Linear}
    alphaScaleBiasClamp {bMapRange=true, inRange {max=50}}
    boneToModify {boneName=glute_volume_l}
    translation {y=-4}
   }
   n33 = Get thigh_rot_r()
   n34 = "Transform (Modify) Bone - Bone: glute_volume_r"(ComponentPose=@n32, Alpha=@n33, Translation="10,2,0", Rotation="0,0,0", Scale="1,1,1") {
    rotationMode=BMM_Additive
    rotationSpace=BCS_BoneSpace
    scaleMode=BMM_Additive
    scaleSpace=BCS_BoneSpace
    translationMode=BMM_Additive
    translationSpace=BCS_BoneSpace
    alphaBoolBlend {blendOption=Linear}
    alphaScaleBiasClamp {bMapRange=true, inRange {max=0, min=-100}, outRange {max=0, min=1}}
    boneToModify {boneName=glute_volume_r}
    translation=<10 2>
   }
   n35 = "Transform (Modify) Bone - Bone: glute_volume_r"(ComponentPose=@n34, Alpha=@n33, Translation="0,4,0", Rotation="0,0,0", Scale="1,1,1") {
    rotationMode=BMM_Additive
    rotationSpace=BCS_BoneSpace
    scaleMode=BMM_Additive
    scaleSpace=BCS_BoneSpace
    translationMode=BMM_Additive
    translationSpace=BCS_BoneSpace
    alphaBoolBlend {blendOption=Linear}
    alphaScaleBiasClamp {bMapRange=true, inRange {max=50}}
    boneToModify {boneName=glute_volume_r}
    translation {y=4}
   }
   n36 = "Constraint - crouch_volume_base_l (PP)"!(ComponentPose=@n35, Alpha=1, ConstraintWeights_0=0.3, ConstraintWeights_1:float=0.7) -> (Pose) {
    alphaBoolBlend {blendOption=Linear}
    boneToModify {boneName=crouch_volume_base_l}
    constraintSetup [{targetBone {boneName=pelvis}}, {targetBone {boneName=thigh_l}}]
    constraintWeights [1,1]
   }
   n37 = "Constraint - crouch_volume_base_r (PP)"!(ComponentPose=@n36, Alpha=1, ConstraintWeights_0=0.3, ConstraintWeights_1:float=0.7) -> (Pose) {
    alphaBoolBlend {blendOption=Linear}
    boneToModify {boneName=crouch_volume_base_r}
    constraintSetup [{targetBone {boneName=pelvis}}, {targetBone {boneName=thigh_r}}]
    constraintWeights [1,1]
   }
   n38 = "Transform (Modify) Bone - Bone: crouch_volume_l"(ComponentPose=@n37, Alpha=Get thigh_rot_l(), Translation="-3,10,0", Rotation="0,0,0", Scale="0.6,1,1") {
    rotationMode=BMM_Additive
    rotationSpace=BCS_BoneSpace
    scaleMode=BMM_Additive
    scaleSpace=BCS_BoneSpace
    translationMode=BMM_Additive
    translationSpace=BCS_BoneSpace
    alphaBoolBlend {blendOption=Linear}
    alphaScaleBiasClamp {bMapRange=true, inRange {max=0, min=-90}, outRange {max=0, min=1}}
    boneToModify {boneName=crouch_volume_l}
    scale {x=0.6}
    translation=<-3 10>
   }
   n39 = "Transform (Modify) Bone - Bone: crouch_volume_r"(ComponentPose=@n38, Alpha=Get thigh_rot_r(), Translation="3,-10,0", Rotation="0,0,0", Scale="0.6,1,1") {
    rotationMode=BMM_Additive
    rotationSpace=BCS_BoneSpace
    scaleMode=BMM_Additive
    scaleSpace=BCS_BoneSpace
    translationMode=BMM_Additive
    translationSpace=BCS_BoneSpace
    alphaBoolBlend {blendOption=Linear}
    alphaScaleBiasClamp {bMapRange=true, inRange {max=0, min=-90}, outRange {max=0, min=1}}
    boneToModify {boneName=crouch_volume_r}
    scale {x=0.6}
    translation=<3 -10>
   }
   n40 = "Constraint - knee_out_volume_base_l (PP)"(ComponentPose=@n39, Alpha=1, ConstraintWeights_0=0.5, ConstraintWeights_1=0.5) {
    alphaBoolBlend {blendOption=Linear}
    boneToModify {boneName=knee_out_volume_base_l}
    constraintSetup [{targetBone {boneName=upperleg_bendy_c_l}}, {targetBone {boneName=lowerleg_bendy_a_l}}]
    constraintWeights [1,1]
   }
   n41 = "Constraint - knee_out_volume_base_r (PP)"(ComponentPose=@n40, Alpha=1, ConstraintWeights_0=0.5, ConstraintWeights_1=0.5) {
    alphaBoolBlend {blendOption=Linear}
    boneToModify {boneName=knee_out_volume_base_r}
    constraintSetup [{targetBone {boneName=lowerleg_bendy_a_r}}, {targetBone {boneName=upperleg_bendy_c_r}}]
    constraintWeights [1,1]
   }
   n42 = "Transform (Modify) Bone - Bone: knee_out_volume_l"(ComponentPose=@n41, Alpha=Get knee_rot_l(), Translation="3,3,0", Rotation="0,0,0", Scale="1.5,1,1.5") {
    rotationMode=BMM_Additive
    rotationSpace=BCS_BoneSpace
    scaleMode=BMM_Additive
    scaleSpace=BCS_BoneSpace
    translationMode=BMM_Additive
    translationSpace=BCS_BoneSpace
    alphaBoolBlend {blendOption=Linear}
    alphaScaleBiasClamp {bMapRange=true, inRange {max=80}}
    boneToModify {boneName=knee_out_volume_l}
    scale {x=1.5, z=1.5}
    translation=<3 3>
   }
   n43 = "Transform (Modify) Bone - Bone: knee_out_volume_r"(ComponentPose=@n42, Alpha=Get knee_rot_r(), Translation="-3,-3,0", Rotation="0,0,0", Scale="1.5,1,1.5") {
    rotationMode=BMM_Additive
    rotationSpace=BCS_BoneSpace
    scaleMode=BMM_Additive
    scaleSpace=BCS_BoneSpace
    translationMode=BMM_Additive
    translationSpace=BCS_BoneSpace
    alphaBoolBlend {blendOption=Linear}
    alphaScaleBiasClamp {bMapRange=true, inRange {max=80}}
    boneToModify {boneName=knee_out_volume_r}
    scale {x=1.5, z=1.5}
    translation=<-3 -3>
   }
   n44 = "Constraint - knee_in_volume_base_l (PP)"(ComponentPose=@n43, Alpha=1, ConstraintWeights_0=0.5, ConstraintWeights_1=0.5) {
    alphaBoolBlend {blendOption=Linear}
    boneToModify {boneName=knee_in_volume_base_l}
    constraintSetup [{targetBone {boneName=upperleg_bendy_c_l}}, {targetBone {boneName=lowerleg_bendy_a_l}}]
    constraintWeights [1,1]
   }
   n45 = "Constraint - knee_in_volume_base_r (PP)"(ComponentPose=@n44, Alpha=1, ConstraintWeights_0=0.5, ConstraintWeights_1=0.5) {
    alphaBoolBlend {blendOption=Linear}
    boneToModify {boneName=knee_in_volume_base_r}
    constraintSetup [{targetBone {boneName=lowerleg_bendy_a_r}}, {targetBone {boneName=upperleg_bendy_c_r}}]
    constraintWeights [1,1]
   }
   n46 = "Transform (Modify) Bone - Bone: knee_in_volume_l"(ComponentPose=@n45, Alpha=Get knee_rot_l(), Translation="0,-10,0", Rotation="0,0,0", Scale="0.5,1,1") {
    rotationMode=BMM_Additive
    rotationSpace=BCS_BoneSpace
    scaleMode=BMM_Additive
    scaleSpace=BCS_BoneSpace
    translationMode=BMM_Additive
    translationSpace=BCS_BoneSpace
    alphaBoolBlend {blendOption=Linear}
    alphaScaleBiasClamp {bMapRange=true, inRange {max=100}}
    boneToModify {boneName=knee_in_volume_l}
    scale {x=0.5}
    translation {y=-10}
   }
   n47 = "Transform (Modify) Bone - Bone: knee_in_volume_r"(ComponentPose=@n46, Alpha=Get knee_rot_r(), Translation="0,10,0", Rotation="0,0,0", Scale="0.5,1,1") {
    rotationMode=BMM_Additive
    rotationSpace=BCS_BoneSpace
    scaleMode=BMM_Additive
    scaleSpace=BCS_BoneSpace
    translationMode=BMM_Additive
    translationSpace=BCS_BoneSpace
    alphaBoolBlend {blendOption=Linear}
    alphaScaleBiasClamp {bMapRange=true, inRange {max=100}}
    boneToModify {boneName=knee_in_volume_r}
    scale {x=0.5}
    translation {y=10}
   }
   n48 = "Constraint - pec_volume_aim_base_l (PPP)"!(ComponentPose=@n47, Alpha=1, ConstraintWeights_0=0.25, ConstraintWeights_1:float=0.55, ConstraintWeights_2:float=0.2) -> (Pose) {
    alphaBoolBlend {blendOption=Linear}
    boneToModify {boneName=pec_volume_aim_base_l}
    constraintSetup [{targetBone {boneName=spine_03}}, {targetBone {boneName=spine_04}}, {targetBone {boneName=spine_05}}]
    constraintWeights [1,1,1]
   }
   n49 = "Constraint - pec_volume_aim_base_r (PPP)"!(ComponentPose=@n48, Alpha=1, ConstraintWeights_0=0.25, ConstraintWeights_1:float=0.55, ConstraintWeights_2:float=0.2) -> (Pose) {
    alphaBoolBlend {blendOption=Linear}
    boneToModify {boneName=pec_volume_aim_base_r}
    constraintSetup [{targetBone {boneName=spine_03}}, {targetBone {boneName=spine_04}}, {targetBone {boneName=spine_05}}]
    constraintWeights [1,1,1]
   }
   n50 = Look At - Bone: pec_volume_aim_l(ComponentPose=@n49, Alpha=1) {
    bUseLookUpAxis=true
    alphaBoolBlend {blendOption=Linear}
    boneToModify {boneName=pec_volume_aim_l}
    lookAtLocation {x=0}
    lookAtTarget {boneReference {boneName=upperarm_l}}
    lookAt_Axis {axis=<1 0>}
    lookUp_Axis {axis {x=0, z=1}}
   }
   n51 = Look At - Bone: pec_volume_aim_r(ComponentPose=@n50, Alpha=1) {
    bUseLookUpAxis=true
    alphaBoolBlend {blendOption=Linear}
    boneToModify {boneName=pec_volume_aim_r}
    lookAtLocation {x=0}
    lookAtTarget {boneReference {boneName=upperarm_r}}
    lookAt_Axis {axis=<-1 0>}
    lookUp_Axis {axis {x=0, z=1}}
   }
   n52 = "Constraint - pec_volume_base_l (PP)"!(ComponentPose=@n51, Alpha=1, ConstraintWeights_0=0.5, ConstraintWeights_1:float=0.5) -> (Pose) {
    alphaBoolBlend {blendOption=Linear}
    boneToModify {boneName=pec_volume_base_l}
    constraintSetup [{targetBone {boneName=pec_volume_aim_l}}, {targetBone {boneName=upperarm_l}}]
    constraintWeights [1,1]
   }
   n53 = "Constraint - pec_volume_base_r (PP)"!(ComponentPose=@n52, Alpha=1, ConstraintWeights_0=0.5, ConstraintWeights_1:float=0.5) -> (Pose) {
    alphaBoolBlend {blendOption=Linear}
    boneToModify {boneName=pec_volume_base_r}
    constraintSetup [{targetBone {boneName=upperarm_r}}, {targetBone {boneName=pec_volume_aim_r}}]
    constraintWeights [1,1]
   }
   n54 = "Constraint - upperarm_up_volume_base_l (PP)"!(ComponentPose=@n53, Alpha=1, ConstraintWeights_0=0.6, ConstraintWeights_1:float=0.4) -> (Pose) {
    alphaBoolBlend {blendOption=Linear}
    boneToModify {boneName=upperarm_up_volume_base_l}
    constraintSetup [{targetBone {boneName=upperarm_twist_01_l}}, {targetBone {boneName=clavicle_l}}]
    constraintWeights [1,1]
   }
   n55 = "Constraint - upperarm_up_volume_base_r (PP)"!(ComponentPose=@n54, Alpha=1, ConstraintWeights_0=0.6, ConstraintWeights_1:float=0.4) -> (Pose) {
    alphaBoolBlend {blendOption=Linear}
    boneToModify {boneName=upperarm_up_volume_base_r}
    constraintSetup [{targetBone {boneName=upperarm_twist_01_r}}, {targetBone {boneName=clavicle_r}}]
    constraintWeights [1,1]
   }
   n56 = "Constraint - upperarm_fr_volume_base_l (PP)"!(ComponentPose=@n55, Alpha=1, ConstraintWeights_0=0.3, ConstraintWeights_1:float=0.7) -> (Pose) {
    alphaBoolBlend {blendOption=Linear}
    boneToModify {boneName=upperarm_fr_volume_base_l}
    constraintSetup [{targetBone {boneName=upperarm_twist_01_l}}, {targetBone {boneName=clavicle_l}}]
    constraintWeights [1,1]
   }
   n57 = "Constraint - upperarm_fr_volume_base_r (PP)"!(ComponentPose=@n56, Alpha=1, ConstraintWeights_0=0.3, ConstraintWeights_1:float=0.7) -> (Pose) {
    alphaBoolBlend {blendOption=Linear}
    boneToModify {boneName=upperarm_fr_volume_base_r}
    constraintSetup [{targetBone {boneName=upperarm_twist_01_r}}, {targetBone {boneName=clavicle_r}}]
    constraintWeights [1,1]
   }
   n58 = "Constraint - upperarm_bk_volume_base_l (PP)"!(ComponentPose=@n57, Alpha=1, ConstraintWeights_0=0.4, ConstraintWeights_1:float=0.6) -> (Pose) {
    alphaBoolBlend {blendOption=Linear}
    boneToModify {boneName=upperarm_bk_volume_base_l}
    constraintSetup [{targetBone {boneName=upperarm_twist_01_l}}, {targetBone {boneName=clavicle_l}}]
    constraintWeights [1,1]
   }
   n59 = "Constraint - upperarm_bk_volume_base_r (PP)"!(ComponentPose=@n58, Alpha=1, ConstraintWeights_0=0.4, ConstraintWeights_1:float=0.6) -> (Pose) {
    alphaBoolBlend {blendOption=Linear}
    boneToModify {boneName=upperarm_bk_volume_base_r}
    constraintSetup [{targetBone {boneName=upperarm_twist_01_r}}, {targetBone {boneName=clavicle_r}}]
    constraintWeights [1,1]
   }
   n60 = "Transform (Modify) Bone - Bone: upperarm_up_volume_l"(ComponentPose=@n59, Alpha=Get psd_arm_up_alpha_l() -> (psd_arm_up_alpha_l:double), Translation="0,0,4", Rotation="0,0,0", Scale="1,1,1") {
    rotationMode=BMM_Additive
    rotationSpace=BCS_BoneSpace
    scaleMode=BMM_Additive
    scaleSpace=BCS_BoneSpace
    translationMode=BMM_Additive
    translationSpace=BCS_BoneSpace
    alphaBoolBlend {blendOption=Linear}
    alphaScaleBiasClamp {inRange {max=100}}
    boneToModify {boneName=upperarm_up_volume_l}
    translation {z=4}
   }
   n61 = "Transform (Modify) Bone - Bone: upperarm_up_volume_r"(ComponentPose=@n60, Alpha=Get psd_arm_up_alpha_r() -> (psd_arm_up_alpha_r:double), Translation="0,0,-4", Rotation="0,0,0", Scale="1,1,1") {
    rotationMode=BMM_Additive
    rotationSpace=BCS_BoneSpace
    scaleMode=BMM_Additive
    scaleSpace=BCS_BoneSpace
    translationMode=BMM_Additive
    translationSpace=BCS_BoneSpace
    alphaBoolBlend {blendOption=Linear}
    alphaScaleBiasClamp {inRange {max=100}}
    boneToModify {boneName=upperarm_up_volume_r}
    translation {z=-4}
   }
   n62 = Get psd_arm_fwd_alpha_l() -> (psd_arm_fwd_alpha_l:double)
   n63 = "Transform (Modify) Bone - Bone: upperarm_fr_volume_l"(ComponentPose=@n61, Alpha=@n62, Translation="0,10,0", Rotation="0,0,0", Scale="1,1,1") {
    rotationMode=BMM_Additive
    rotationSpace=BCS_BoneSpace
    scaleMode=BMM_Additive
    scaleSpace=BCS_BoneSpace
    translationMode=BMM_Additive
    translationSpace=BCS_BoneSpace
    alphaBoolBlend {blendOption=Linear}
    alphaScaleBiasClamp {inRange {max=100}}
    boneToModify {boneName=upperarm_fr_volume_l}
    translation {y=10}
   }
   n64 = "Transform (Modify) Bone - Bone: upperarm_bk_volume_l"(ComponentPose=@n63, Alpha=@n62, Translation="0,-6,0", Rotation="0,0,0", Scale="1,1,1") {
    rotationMode=BMM_Additive
    rotationSpace=BCS_BoneSpace
    scaleMode=BMM_Additive
    scaleSpace=BCS_BoneSpace
    translationMode=BMM_Additive
    translationSpace=BCS_BoneSpace
    alphaBoolBlend {blendOption=Linear}
    alphaScaleBiasClamp {inRange {max=100}}
    boneToModify {boneName=upperarm_bk_volume_l}
    translation {y=-6}
   }
   n65 = "Transform (Modify) Bone - Bone: pec_volume_l"(
    ComponentPose=@n64
    Alpha=Get psd_arm_fwd_dn_alpha_l() -> (psd_arm_fwd_dn_alpha_l:double)
    Translation="-3,0,0"
    Rotation="0,0,0"
    Scale="1,1,1"
   ) {
    rotationMode=BMM_Additive
    rotationSpace=BCS_BoneSpace
    scaleMode=BMM_Additive
    scaleSpace=BCS_BoneSpace
    translationMode=BMM_Additive
    translationSpace=BCS_BoneSpace
    alphaBoolBlend {blendOption=Linear}
    alphaScaleBiasClamp {inRange {max=100}}
    boneToModify {boneName=pec_volume_l}
    translation {x=-3}
   }
   n66 = Get psd_arm_fwd_alpha_r() -> (psd_arm_fwd_alpha_r:double)
   n67 = "Transform (Modify) Bone - Bone: upperarm_fr_volume_r"(ComponentPose=@n65, Alpha=@n66, Translation="0,-10,0", Rotation="0,0,0", Scale="1,1,1") {
    rotationMode=BMM_Additive
    rotationSpace=BCS_BoneSpace
    scaleMode=BMM_Additive
    scaleSpace=BCS_BoneSpace
    translationMode=BMM_Additive
    translationSpace=BCS_BoneSpace
    alphaBoolBlend {blendOption=Linear}
    alphaScaleBiasClamp {inRange {max=100}}
    boneToModify {boneName=upperarm_fr_volume_r}
    translation {y=-10}
   }
   n68 = "Transform (Modify) Bone - Bone: upperarm_bk_volume_r"(ComponentPose=@n67, Alpha=@n66, Translation="0,6,0", Rotation="0,0,0", Scale="1,1,1") {
    rotationMode=BMM_Additive
    rotationSpace=BCS_BoneSpace
    scaleMode=BMM_Additive
    scaleSpace=BCS_BoneSpace
    translationMode=BMM_Additive
    translationSpace=BCS_BoneSpace
    alphaBoolBlend {blendOption=Linear}
    alphaScaleBiasClamp {inRange {max=100}}
    boneToModify {boneName=upperarm_bk_volume_r}
    translation {y=6}
   }
   n69 = "Transform (Modify) Bone - Bone: pec_volume_r"(
    ComponentPose=@n68
    Alpha=Get psd_arm_fwd_dn_alpha_r() -> (psd_arm_fwd_dn_alpha_r:double)
    Translation="3,0,0"
    Rotation="0,0,0"
    Scale="1,1,1"
   ) {
    rotationMode=BMM_Additive
    rotationSpace=BCS_BoneSpace
    scaleMode=BMM_Additive
    scaleSpace=BCS_BoneSpace
    translationMode=BMM_Additive
    translationSpace=BCS_BoneSpace
    alphaBoolBlend {blendOption=Linear}
    alphaScaleBiasClamp {inRange {max=100}}
    boneToModify {boneName=pec_volume_r}
    translation {x=3}
   }
   n70 = "Transform (Modify) Bone - Bone: upperarm_bk_volume_l"(ComponentPose=@n69, Alpha=Get psd_arm_bk_alpha_l() -> (psd_arm_bk_alpha_l:double), Translation="0,-4,0", Rotation="0,0,0", Scale="1,1,1") {
    rotationMode=BMM_Additive
    rotationSpace=BCS_BoneSpace
    scaleMode=BMM_Additive
    scaleSpace=BCS_BoneSpace
    translationMode=BMM_Additive
    translationSpace=BCS_BoneSpace
    alphaBoolBlend {blendOption=Linear}
    alphaScaleBiasClamp {inRange {max=100}}
    boneToModify {boneName=upperarm_bk_volume_l}
    translation {y=-4}
   }
   n71 = "Transform (Modify) Bone - Bone: upperarm_bk_volume_l"(ComponentPose=@n70, Alpha=Get psd_arm_bk_alpha_r() -> (psd_arm_bk_alpha_r:double), Translation="0,4,0", Rotation="0,0,0", Scale="1,1,1") {
    rotationMode=BMM_Additive
    rotationSpace=BCS_BoneSpace
    scaleMode=BMM_Additive
    scaleSpace=BCS_BoneSpace
    translationMode=BMM_Additive
    translationSpace=BCS_BoneSpace
    alphaBoolBlend {blendOption=Linear}
    alphaScaleBiasClamp {inRange {max=100}}
    boneToModify {boneName=upperarm_bk_volume_l}
    translation {y=4}
   }
   n72 = "Constraint - bicep_volume_base_l (PPP)"!(ComponentPose=@n71, Alpha=1, ConstraintWeights_0=0.5, ConstraintWeights_1:float=0.3, ConstraintWeights_2:float=0.2) -> (Pose) {
    alphaBoolBlend {blendOption=Linear}
    boneToModify {boneName=bicep_volume_base_l}
    constraintSetup [{targetBone {boneName=upperarm_bendy_c_l}}, {targetBone {boneName=upperarm_bendy_b_l}}, {targetBone {boneName=upperarm_twist_01_l}}]
    constraintWeights [1,1,1]
   }
   n73 = "Constraint - bicep_volume_base_r (PPP)"!(ComponentPose=@n72, Alpha=1, ConstraintWeights_0=0.2, ConstraintWeights_1:float=0.5, ConstraintWeights_2:float=0.3) -> (Pose) {
    alphaBoolBlend {blendOption=Linear}
    boneToModify {boneName=bicep_volume_base_r}
    constraintSetup [{targetBone {boneName=upperarm_twist_01_r}}, {targetBone {boneName=upperarm_bendy_c_r}}, {targetBone {boneName=upperarm_bendy_b_r}}]
    constraintWeights [1,1,1]
   }
   Save cached pose 'volumePose'(Pose=Component To Local(ComponentPose=@n73)) {cacheName=volumePose}
   n74 = Layered blend per bone(
    BasePose=Copy Pose From Mesh() {bCopyCurves=true, bUseAttachedParent=true}
    BlendPoses_0="Sequence Evaluator (by time) 'Emote_BuffCatComic_CMM_Bends'"!(ExplicitTime:float=Get emoteOverwriteJoint_Position() -> (emoteOverwriteJoint_Position:double)) -> (Pose:Pose Link Structure) {is=title AnimGraphNode_SequenceEvaluator, sequence=/Game/Animation/Game/MainPlayer/Emotes/BuffCatComic/Emote_BuffCatComic_CMM_Bends}
    BlendWeights_0=Get EmoteAlpha() -> (EmoteAlpha:double)
   ) {layerSetup [{branchFilters [
       {boneName=upperarm_bendy_b_l}
       {boneName=upperarm_bendy_b_l}
       {boneName=upperarm_bendy_c_l}
       {boneName=upperarm_bendy_c_r}
       {boneName=lowerarm_bendy_a_l}
       {boneName=lowerarm_bendy_b_l}
       {boneName=lowerarm_bendy_c_l}
       {boneName=lowerarm_bendy_a_r}
       {boneName=lowerarm_bendy_b_r}
       {boneName=lowerarm_bendy_c_r}
       {boneName=upperleg_bendy_b_l}
       {boneName=upperleg_bendy_c_l}
       {boneName=upperleg_bendy_b_r}
       {boneName=upperleg_bendy_c_r}
       {boneName=lowerleg_bendy_a_l}
       {boneName=lowerleg_bendy_b_l}
       {boneName=lowerleg_bendy_c_l}
       {boneName=lowerleg_bendy_a_r}
       {boneName=lowerleg_bendy_b_r}
       {boneName=lowerleg_bendy_c_r}
       {boneName=bicep_volume_l}
       {boneName=bicep_volume_r}
      ]}]}
   n75 = Control Rig!(
    Source=@n74
    Alpha
    elbow_rot_l:double=Get elbow_rot_l()
    elbow_rot_r:double=Get elbow_rot_r()
    thigh_rot_l:double=Get thigh_rot_l()
    thigh_rot_r:double=Get thigh_rot_r()
    knee_rot_l:double=Get knee_rot_l()
    knee_rot_r:double=Get knee_rot_r()
    psd_arm_up_alpha_l:double=Get psd_arm_up_alpha_l()
    psd_arm_up_alpha_r:double=Get psd_arm_up_alpha_r()
    psd_arm_fwd_alpha_l:double=Get psd_arm_fwd_alpha_l()
    psd_arm_fwd_alpha_r:double=Get psd_arm_fwd_alpha_r()
    psd_arm_fwd_dn_alpha_l:double=Get psd_arm_fwd_dn_alpha_l()
    psd_arm_fwd_dn_alpha_r:double=Get psd_arm_fwd_dn_alpha_r()
    psd_arm_bk_alpha_l:double=Get psd_arm_bk_alpha_l()
    psd_arm_bk_alpha_r:double=Get psd_arm_bk_alpha_r()
   ) -> (Pose) {
    alphaBoolBlend {blendOption=Linear}
    controlRigAssetReference {blueprintRigClass=&/volume_ControlRig.volume_ControlRig_C}
   }
   Save cached pose 'EmoteBlend_pose'(Pose=@n75) {cacheName=EmoteBlend_pose}
   Use cached pose 'volumePose'()
   Use cached pose 'EmoteBlend_pose'()
  ]}
 EventGraph {nodes [
   n1 = Event BlueprintUpdateAnimation()
   > n2 = Sequence!(execute) -> (then_0, then_1, then_2:Exec, then_3:Exec, then_4:Exec, then_5:Exec, then_6:Exec)
   > n3 = Sequence()
   n4 = GetDeltaTransformFromRefPose(self=GetOwnerSkeletalMeshComponent(), BoneName=lowerarm_l, BaseName=upperarm_l)
   > Set elbow_rot_l(elbow_rot_l:double=BreakRotator(InRot=BreakTransform(InTransform=@n4).Rotation).Yaw) -> (Output_Get:double)
   n5 = GetDeltaTransformFromRefPose(self=GetOwnerSkeletalMeshComponent(), BoneName=lowerarm_r, BaseName=upperarm_r)
   > Set elbow_rot_r(elbow_rot_r:double=BreakRotator(InRot=BreakTransform(InTransform=@n5).Rotation).Yaw) -> (Output_Get:double)
   n6 = GetDeltaTransformFromRefPose(self=GetOwnerSkeletalMeshComponent(), BoneName=calf_l, BaseName=thigh_l)
   Set knee_rot_l(execute=@n3.then_2, knee_rot_l:double=BreakRotator(InRot=BreakTransform(InTransform=@n6).Rotation).Yaw) -> (Output_Get:double)
   n7 = GetDeltaTransformFromRefPose(self=GetOwnerSkeletalMeshComponent(), BoneName=calf_r, BaseName=thigh_r)
   > Set knee_rot_r(knee_rot_r:double=BreakRotator(InRot=BreakTransform(InTransform=@n7).Rotation).Yaw) -> (Output_Get:double)
   n8 = GetDeltaTransformFromRefPose(self=GetOwnerSkeletalMeshComponent(), BoneName=thigh_l, BaseName=pelvis)
   Set thigh_rot_l(execute=@n3.then_1, thigh_rot_l=BreakRotator(InRot=BreakTransform(InTransform=@n8).Rotation).Yaw)
   n9 = GetDeltaTransformFromRefPose(self=GetOwnerSkeletalMeshComponent(), BoneName=thigh_r, BaseName=pelvis)
   > Set thigh_rot_r(thigh_rot_r=BreakRotator(InRot=BreakTransform(InTransform=@n9).Rotation).Yaw)
   n10 = Sequence(execute=@n2.then_1)
   > n11 = PSD Func!(execute, self, startJnt:String=upperarm_l, driverJnt:String=lowerarm_l, poseJnt:String=pose_arm_up_l, angle:double=70, debug:bool=false) -> (then, alpha:double)
   > Set psd_arm_up_alpha_l(psd_arm_up_alpha_l:double=@n11) -> (Output_Get:double)
   > n12 = PSD Func(startJnt=upperarm_r, driverJnt=lowerarm_r, poseJnt=pose_arm_up_r, angle=70, debug=false)
   > Set psd_arm_up_alpha_r(psd_arm_up_alpha_r:double=@n12) -> (Output_Get:double)
   n13 = PSD Func(execute=@n10.then_1, startJnt=upperarm_l, driverJnt=lowerarm_l, poseJnt=pose_arm_fwd_l, angle=50, debug=false)
   > Set psd_arm_fwd_alpha_l(psd_arm_fwd_alpha_l:double=@n13) -> (Output_Get:double)
   > n14 = PSD Func(startJnt=upperarm_r, driverJnt=lowerarm_r, poseJnt=pose_arm_fwd_r, angle=50, debug=false)
   > Set psd_arm_fwd_alpha_r(psd_arm_fwd_alpha_r:double=@n14) -> (Output_Get:double)
   n15 = PSD Func(execute=@n10.then_2, startJnt=upperarm_l, driverJnt=lowerarm_l, poseJnt=pose_arm_fwd_dn_l, angle=30, debug=false)
   > Set psd_arm_fwd_dn_alpha_l(psd_arm_fwd_dn_alpha_l:double=@n15) -> (Output_Get:double)
   > n16 = PSD Func(startJnt=upperarm_r, driverJnt=lowerarm_r, poseJnt=pose_arm_fwd_dn_r, angle=30, debug=false)
   > Set psd_arm_fwd_dn_alpha_r(psd_arm_fwd_dn_alpha_r:double=@n16) -> (Output_Get:double)
   n17 = PSD Func(execute=@n10.then_3, startJnt=upperarm_l, driverJnt=lowerarm_l, poseJnt=pose_arm_bk_l, angle=50, debug=false)
   > Set psd_arm_bk_alpha_l(psd_arm_bk_alpha_l:double=@n17) -> (Output_Get:double)
   > n18 = PSD Func(startJnt=upperarm_r, driverJnt=lowerarm_r, poseJnt=pose_arm_bk_r, angle=50, debug=false)
   > Set psd_arm_bk_alpha_r(psd_arm_bk_alpha_r:double=@n18) -> (Output_Get:double)
   n19 = Branch(execute=@n10.then_4, Condition=Get bIsCrouching())
   > Set isCrouchingAlpha(isCrouchingAlpha:double=1) -> (Output_Get:double)
   Set isCrouchingAlpha(execute=@n19.else, isCrouchingAlpha=0)
   n20 = Get Anim_Inst() -> (Anim_Inst:Anim Instance Object Reference)
   n21 = Montage_IsActive!(
    self:Anim Instance Object Reference=@n20
    Montage:Anim Montage Object Reference=/Game/Animation/Game/MainPlayer/Emotes/BuffCatComic/Emote_BuffCatComic_CMM_M
   ) -> (ReturnValue:bool)
   Set EmoteAlpha(
    execute=@n10.then_5
    EmoteAlpha:double=FInterpTo(Current=Get EmoteAlpha(), Target="To Float (Boolean)"(InBool=@n21), DeltaTime=@n1.DeltaTimeX, InterpSpeed=7)
   ) -> (Output_Get:double)
   > Set emoteOverwriteJoint_Position(emoteOverwriteJoint_Position:double=Montage_GetPosition!(
     self:Anim Instance Object Reference=@n20
     Montage:Anim Montage Object Reference=&Emote_BuffCatComic_CMM_M
    ) -> (ReturnValue:float)) -> (Output_Get:double)
   Event OnCharacterCustomizationLoaded()
   > Set BodyMesh(BodyMesh=GetOwnerSkeletalMeshComponent())
   > n22 = Cast To PlayerPawn_Athena(Object=TryGetPawnOwner())
   > n23 = Set PawnOwner(PawnOwner:Player Pawn Athena Object Reference=@n22) -> (Output_Get:Player Pawn Athena Object Reference)
   > Set TailMesh(TailMesh:Skeletal Mesh Component Object Reference=GetSkeletalMeshForPartType(self=@n23, PartType=MiscOrTail)) -> (Output_Get:Skeletal Mesh Component Object Reference)
   n24 = CheckForBackbling(LongBackblingThreshold=40)
   > Set isWearingBackbling(isWearingBackbling=@n24)
   > Set isWearingCape(isWearingCape=@n24.bIsBackblingLong)
   > Set characterHasTail(characterHasTail:bool=CheckForTail!(self:Custom Character Part Anim Instance Object Reference) -> (bHasTail:bool)) -> (Output_Get:bool)
   > Branch(Condition=Get characterHasTail() -> (characterHasTail:bool))
   > n25 = Branch(Condition=Get isWearingCape() -> (isWearingCape:bool))
   n26 = Get TailMesh() -> (TailMesh:Skeletal Mesh Component Object Reference)
   > SetVisibility!(execute, self:Scene Component Object Reference=@n26, bNewVisibility:bool=false, bPropagateToChildren:bool=false) -> (then)
   SetVisibility(execute=@n25.else, self=@n26, bNewVisibility=true, bPropagateToChildren=false)
   Event BlueprintInitializeAnimation()
   n27 = GetOwnerSkeletalMeshComponent()
   > Branch(Condition=IsValid(Object=@n27))
   > n28 = Set Skel_Mesh(Skel_Mesh=@n27)
   > Set Anim_Inst(Anim_Inst:Anim Instance Object Reference=GetAnimInstance(self=@n28)) -> (Output_Get:Anim Instance Object Reference)
  ]}
 PSD Func {nodes [
   n1 = PSD Func!() -> (then, startJnt:String=upperarm_l, driverJnt:String=lowerarm_l, poseJnt:String=dyn_arm_measure_l, angle:double=45, debug:bool=~) {is=title K2Node_FunctionEntry}
   > n2 = Sequence!(execute) -> (then_0, then_1)
   > Branch(Condition=@n1.debug)
   n3 = GetPartSkeletalMeshComponent()
   n4 = GetSocketLocation!(self:Scene Component Object Reference=@n3, InSocketName:Name=String To Name(InString=@n1)) -> (ReturnValue:Vector)
   > DrawDebugArrow!(
    execute
    LineStart:Vector=@n4
    LineEnd:Vector=GetSocketLocation(self=@n3, InSocketName=String To Name(InString=@n1.poseJnt))
    ArrowSize:float=0
    LineColor:Linear Color Structure=(R=0.1692,G=0.147,B=1,A=1)
    Duration:float=0
    Thickness:float=1
    DepthPriority:EDrawDebugSceneDepthPriorityGroup Enum=World
   ) -> (then)
   n5 = GetPartSkeletalMeshComponent()
   n6 = Get Direction Between Sockets!(
    Component:Skeletal Mesh Component Object Reference=@n5
    SocketOrBoneNameFrom:Name=String To Name(InString=@n1)
    SocketOrBoneNameTo:Name=String To Name(InString=@n1.poseJnt)
   ) -> (ReturnValue:Vector) {is=title K2Node_CallFunction}
   > Draw Debug Cone!(
    execute
    Origin:Vector=@n4
    Direction:Vector=@n6
    Length:float=100
    AngleWidth:float=@n1.angle
    AngleHeight:float=@n1.angle
    NumSides:int=12
    LineColor:Linear Color Structure=(R=1,G=1,B=1,A=1)
    Duration:float=0
    Thickness:float=0.3
    DepthPriority:EDrawDebugSceneDepthPriorityGroup Enum=World
   ) -> (then)
   n7 = Get Direction Between Sockets(Component=@n5, SocketOrBoneNameFrom=String To Name(InString=@n1), SocketOrBoneNameTo=String To Name(InString=@n1.driverJnt))
   n8 = float - float(B=float / float(A=GetAngleBetweenTwoVectors!(VectorA:Vector=@n7, VectorB:Vector=@n6) -> ("Angle (Degrees)":double), B=@n1.angle) {is=title K2Node_CallFunction}) {is=title K2Node_CallFunction}
   n9 = Branch(execute=@n2.then_1, Condition="float >= float"(A=@n8, B=0) {is=title K2Node_CallFunction})
   > n10 = Set psd_alpha(psd_alpha:double=@n8) -> (Output_Get:double)
   Return Node!(execute=[@n11.then, @n10.then], alpha:double=@n10)
   n11 = Set psd_alpha(execute=@n9.else, psd_alpha=0)
  ]}
]
variables [
 Anim_Inst {type=AnimInstance, default=null}
 BodyMesh {type=SkeletalMeshComponent}
 characterHasTail {type=boolean}
 elbow_rot_l {type=number}
 elbow_rot_r {type=number}
 EmoteAlpha {type=number}
 emoteOverwriteJoint_Position {type=number}
 isCrouchingAlpha {type=number}
 isWearingBackbling {type=boolean}
 isWearingCape {type=boolean}
 knee_rot_l {type=number}
 knee_rot_r {type=number}
 PawnOwner {type=PlayerPawn_Athena_C}
 psd_arm_bk_alpha_l {type=number}
 psd_arm_bk_alpha_r {type=number}
 psd_arm_fwd_alpha_l {type=number}
 psd_arm_fwd_alpha_r {type=number}
 psd_arm_fwd_dn_alpha_l {type=number}
 psd_arm_fwd_dn_alpha_r {type=number}
 psd_arm_up_alpha_l {type=number}
 psd_arm_up_alpha_r {type=number}
 Skel_Mesh {type=SkeletalMeshComponent}
 TailMesh {type=SkeletalMeshComponent}
 thigh_rot_l {type=number}
 thigh_rot_r {type=number}
]
